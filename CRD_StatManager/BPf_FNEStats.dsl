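; Blueprint function library that resolves a stat event filter to its display details, such as icon, color, matching minigame stat filter and scope.
; GetStatDetails searches the PlayerStatDetails playspace component's map, testing both stat and level filters.
; Fallbacks query BPi_StatDetails actors or a local class-to-visuals map, then find the loaded stat of that class.
; The visuals lookup compares event and tag queries against class defaults because the editor cannot access loaded stats.
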(fn FilterMatchesStat
  (params (Minigame%20filter :type Fort%20Minigame%20Stat%20Filter%20Object%20Reference) (StatEventFilter :type Stat%20Event%20Filter%20Structure%20%28by%20ref%29))
  (bind (_statevent _itemdefinition _targettagsquery _sourcetagsquery _contexttagsquery) (Utilities|Struct|BreakStatEventFilter StatEventFilter))
  (Utilities|IsValid Minigame%20filter
    (:Is%20Valid
      (return (Class|FortMinigameStatFilter|Matches Minigame%20filter _statevent (Utilities|ResolveSoftReference _itemdefinition) 0 _targettagsquery _sourcetagsquery _contexttagsquery)))
    (:Is%20Not%20Valid
      (return false))))

(fn FindLoadedStatByClass
  (params (Class :type Object%20Class%20Reference) (ContextActor :type Actor%20Object%20Reference))
  (bind _returnvalue (Minigame|Stats|GetMinigameStatsComponent ContextActor))
  (Utilities|IsValid _returnvalue
    (:Is%20Valid
      (for _array_element (Variables|/Script/FortniteGame.FortMinigameStatsComponent|GetLoadedStats _returnvalue)
        (if (== (Utilities|GetClass _array_element) Class)
          (return _array_element)))
      (return 0))))

(fn FindStatPlayspaceComponent
  (params (Context :type Actor%20Object%20Reference))
  (bind _returnvalue (Class|PlayspaceLibrary|GetPlayspaceforActor Context Context))
  (Utilities|IsValid _returnvalue
    (:Is%20Valid
      (return (Actor|GetComponentbyClass _returnvalue "/CRD_StatManager/UI/PlayspaceComponent_PlayerStatDetails.PlayspaceComponent_PlayerStatDetails_C")))
    (:Is%20Not%20Valid
      (return 0))))

(fn GetDetailsforMinigameStat
  (params (StatEventFilter :type Stat%20Event%20Filter%20Structure%20%28by%20ref%29))
  (bind (_statevent _itemdefinition _targettagsquery _sourcetagsquery _contexttagsquery) (Utilities|Struct|BreakStatEventFilter StatEventFilter))
  (bind _value (Utilities|Map|Find (LocalVariables|GetStatVisuals) (LocalVariables|GetTestClass)))
  (bind (_value_texture_2_dbf425c84f4c66ce82aca09d37ee8463 _value_color_5_ac0dd7e94a333e0efaa004b301912d62) (Utilities|Struct|BreakTextureColorPair _value))
  (bind (_statevent_2 _targettagsquery_2 _sourcetagsquery_2 _contexttagsquery_2) (Class|GetClassDefaults (LocalVariables|GetTestClass)))
  (bind _keys (Utilities|Map|Keys (LocalVariables|GetStatVisuals))) ; Editor can't look up loaded stats,\nso needs to check against CDOs 
  (for _array_element _keys
    (LocalVariables|SetTestClass _array_element)
    (if (Math|Boolean|ANDBoolean (Utilities|Enum|Equal(Enum) _statevent _statevent_2) (or (GameplayTags|IsTagQueryEmpty _targettagsquery_2) (GameplayTags|DoesContainerMatchTagQuery _targettagsquery _targettagsquery_2)) (or (GameplayTags|IsTagQueryEmpty _sourcetagsquery_2) (GameplayTags|DoesContainerMatchTagQuery _sourcetagsquery _sourcetagsquery_2)) (or (GameplayTags|IsTagQueryEmpty _contexttagsquery_2) (GameplayTags|DoesContainerMatchTagQuery _contexttagsquery _contexttagsquery_2)))
      (return true _value_texture_2_dbf425c84f4c66ce82aca09d37ee8463 _value_color_5_ac0dd7e94a333e0efaa004b301912d62 (LocalVariables|GetTestClass))))
  (return false "" "(R=0.197254,G=0.208333,B=0.194608,A=1.000000)" 0))

(fn GetStatDetails
  (params (StatEventFilter :type Stat%20Event%20Filter%20Structure%20%28by%20ref%29) (WorldContext :type Actor%20Object%20Reference))
  (bind _worldcontext (Variables|GetWorldContext))
  (bind _worldcontext_2 (Variables|GetWorldContext))
  (if (not (Editor|InEditor))
    (bind _comp (CallFunction|FindStatPlayspaceComponent WorldContext WorldContext))
    (Utilities|IsValid _comp
      (:Is%20Valid
        (Utilities|IsValid (Minigame|Stats|GetMinigameStatsComponent WorldContext)
          (:Is%20Valid
            (bind _statdetails (Class|PlayspaceComponentPlayerStatDetails|GetStatDetails _comp))
            (bind _keys (Utilities|Map|Keys _statdetails))
            (for _array_element _keys
              (bind _output_get (LocalVariables|SetFoundDetails (Utilities|Map|Find _statdetails _array_element)))
              (if (CallFunction|FilterMatchesStat (Utilities|Struct|BreakStatCreatorDetails _output_get) (Variables|GetStatEventFilter) (Variables|GetWorldContext))
                (return true (.Icon_45_BBE7BF714C8231A755EF1A968D245643 (Utilities|Struct|BreakStatCreatorDetails _output_get)) (.StatColor_7_0619FD9D48574A0223759FA4571609FA (Utilities|Struct|BreakStatCreatorDetails _output_get)) (Utilities|Struct|BreakStatCreatorDetails _output_get) (.Scope_50_B1C6C3F34ED58F1C2DBF459BD39ACF13 (Utilities|Struct|BreakStatCreatorDetails _output_get)))
                (elif (CallFunction|FilterMatchesStat (.LevelFilter_54_50B3735347949DE5ACEA8BA3543EE1F6 (Utilities|Struct|BreakStatCreatorDetails _output_get)) (Variables|GetStatEventFilter) (Variables|GetWorldContext))
                  (return true (.Icon_45_BBE7BF714C8231A755EF1A968D245643 (Utilities|Struct|BreakStatCreatorDetails _output_get)) (.StatColor_7_0619FD9D48574A0223759FA4571609FA (Utilities|Struct|BreakStatCreatorDetails _output_get)) (.LevelFilter_54_50B3735347949DE5ACEA8BA3543EE1F6 (Utilities|Struct|BreakStatCreatorDetails _output_get)) (.Scope_50_B1C6C3F34ED58F1C2DBF459BD39ACF13 (Utilities|Struct|BreakStatCreatorDetails _output_get)))))))
          (:Is%20Not%20Valid
            (if (Editor|InEditor)
              (bind _outactors (Actor|GetAllActorswithInterface (Variables|GetWorldContext) "/CRD_StatManager/BPi_StatDetails.BPi_StatDetails_C"))
              (for _array_element_2 _outactors
                (bind (_matches _icon _color) (StatDetails|GetStatDetails(Message) _array_element_2 (Variables|GetStatEventFilter)))
                (if _matches
                  (return true _icon _color 0 "NewEnumerator0")))
              (bind (_found _icon_2 _color_2 _stat_class) (CallFunction|GetDetailsForMinigameStat (Variables|GetStatEventFilter) _worldcontext))
              (return _found _icon_2 _color_2 (CallFunction|FindLoadedStatbyClass _stat_class (Variables|GetWorldContext) _worldcontext) "NewEnumerator0")))))
      (:Is%20Not%20Valid
        _))
    (bind (_found_2 _icon_3 _color_3 _stat_class_2) (CallFunction|GetDetailsForMinigameStat (Variables|GetStatEventFilter) _worldcontext_2))
    (return _found_2 _icon_3 _color_3 (CallFunction|FindLoadedStatbyClass _stat_class_2 (Variables|GetWorldContext) _worldcontext_2) "NewEnumerator0")
    (else
      _)))
