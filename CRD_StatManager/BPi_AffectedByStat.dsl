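; BPi_AffectedByStat is a Blueprint interface in the CRD_StatManager folder for objects that react to stat changes.
; It declares one function, UpdateStat, which passes a stat gameplay tag, a validity flag, a soft icon texture reference, a color and a display name.
; The file holds only the signature with no return values, so each implementer defines its own response.
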(fn UpdateStat
  (params (Stat%20Tag :type Gameplay%20Tag%20Structure) (IsValid :type Boolean) (Icon :type Texture%202D%20Soft%20Object%20Reference) (Color :type Linear%20Color%20Structure) (Stat%20name :type Text))
  (return))
