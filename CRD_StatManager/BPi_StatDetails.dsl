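; BPi_StatDetails is a Blueprint interface in the CRD_StatManager folder for querying details about a stat.
; It declares one function, GetStatDetails, which takes a Stat Name input typed as a Stat Event Filter Structure.
; The function returns three values: a boolean, a text value and a number.
; The file holds only default return values, with no lookup logic.
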(fn GetStatDetails
  (params (Stat%20Name :type Stat%20Event%20Filter%20Structure))
  (return false "" 0))
