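; Creative powerup device that applies a chosen stat by magnitude when picked up.
; Effect setup tags the gameplay effect with the stat and whether it is instant or timed.
; Stat icon and color, falling back to class defaults, drive mesh material and VFX, refreshing from the details component or, in editor, Stat Creators.
; A separate overlap component triggers pickup unless a radius override is set, avoiding root scaling issues.
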(event Custom|Get%20And%20Update%20Stat%20Details
  (bind (_stattoapply _statcolor _staticon) (Class|GetClassDefaults "/CRD_StatManager/Device_Powerup_Stat.Device_Powerup_Stat_C"))
  ;[1 0.6099 0 0.2] Gathering the icon and color for the stat from the Player Controller Component (or the Device directly in editor) then loading the color)
  (bind (_found _stat_icon_large_icon _statcolor_2) (Class|BPfFNEStats|GetStatDetails (Variables|UserOptions|Basic|GetStatToApply) self))
  (if _found
    (Variables|Default|SetStatColor _statcolor_2)
    (Variables|Default|SetStatIcon _stat_icon_large_icon)
    (else
      (Variables|Default|SetStatColor _statcolor)
      (Variables|Default|SetStatIcon _staticon)))
  (CallFunction|UpdateStatDetails))

(event EventUpdateForPropertyChange
  (bind _stattoapply_statevent (Utilities|Struct|BreakStatEventFilter (Variables|GetStatToApply_StatEvent)))
  ;[0.04337 1 0.127 0.187] Update the visuals on the powerup when we change any value (gathers from Stat Details Player Controller component, or directly from the Stat Creator if we're in editor)
  (|CallParentFunction)
  (CallFunction|GetandUpdateStatDetails)
  (Variables|Default|SetEffect%20Level (Math|Conversions|ToFloat(Byte) _stattoapply_statevent)))

(event Custom|Update%20Stat%20Details
  (bind _meshmid (Variables|Default|GetMeshMID))
  (bind _meshmid_2 (Variables|Default|GetMeshMID))
  ;[1 0.6099 0 0.2] Gathering the icon and color for the stat from the Player Controller Component (or the Device directly in editor) then loading the color)
  (Variables|Default|SetVFX%20Particle%20Color (Variables|Default|GetStatColor))
  (bind _staticon (Variables|Default|GetStatIcon)
    (:then
      (bind _loaded_object (Creative|Loading|AsyncLoadEditorSafe _staticon)
        (:Loaded
          (bind _astexture_2d (Utilities|Casting|CastToTexture2D _loaded_object))
          (Material|ReattachOrCreateMID (Variables|Default|GetMesh) _meshmid_2)
          (Rendering|Material|SetTextureParameterValue _meshmid_2 "Logo" _astexture_2d))))
    (:else
      (Material|ReattachOrCreateMID (Variables|Default|GetMesh) _meshmid)
      (Rendering|Material|SetTextureParameterValue _meshmid "Logo" "")))
  (Rendering|Material|SetVectorParameterValue _meshmid_2 "Color" (Variables|Default|GetStatColor) :!class /Script/Engine.MaterialInstanceDynamic))

(event Custom|CreatorStatUpdated
  (params (New%20Details :type Map%20of%20Fort%20Minigame%20Stat%20Filter%20Object%20References%20to%20Stat%20Creator%20Details%20Structures))
  ;[1 0.6099 0 0.2] Any Creator Stat has updated. This is called from the Player Controller Component. Used to check if we need to update our version.
  (CallFunction|GetandUpdateStatDetails))

(event OnComponentBeginOverlap(PickupOverlap)
  (params (OverlappedComponent :type Primitive%20Component%20Object%20Reference) (OtherActor :type Actor%20Object%20Reference) (OtherComp :type Primitive%20Component%20Object%20Reference) (OtherBodyIndex :type Integer) (bFromSweep :type Boolean) (SweepResult :type Hit%20Result%20Structure%20%28by%20ref%29))
  ;[0.04337 1 0.127 0.187] Pickup on an overlap component  (not using the root component because we want to have a different scale, and we can't scale the root component and still have it play nice with the phone tool. Also because the visual mesh doesn't match the collision due to World Position Offset in the material)
  (if (not (Variables|UserOptions|GetPickupRadius_Override))
    (CallFunction|DoPickup OtherActor)))

(fn GetEffectMagnitude
  (return (Math|Conversions|ToFloat(Integer) (Variables|UserOptions|Basic|GetMagnitude))))

(fn GetCreativeActorBounds
  (bind (_origin _boxextent _sphereradius) (Collision|GetComponentBounds (Variables|Default|GetPickupOverlap)))
  (return true (/ _sphereradius 2.0) (/ _sphereradius 2.0) (/ _sphereradius 2.0)))

(fn GetCreativeActorOrigin
  (return false 0.0 0.0 0.0))

(fn GetCreativeActorRelativeTransform
  (return false 0))

(fn GetGridAxisGuideAttachmentPoint
  (return true true "0.000000,0.000000,-180.000000"))

(fn IsInstant
  (return (Math|Float|NearlyEqual(Float) (CallFunction|GetDuration) 0.1)))

(fn OnRep_StatDetailsComponent
  (bind _statdetailscomponent (Variables|Default|GetStatDetailsComponent)
    (:then
      (Default|BindEventtoOnClientDetailsUpdated _statdetailscomponent (EventDispatchers|CreateEvent :!event CreatorStatUpdated))
      (CallFunction|GetandUpdateStatDetails))
    (:else)))

(fn SetupEffectData
  (params (EffectHandle :type Gameplay%20Effect%20Spec%20Handle%20Structure) (EffectClass :type Gameplay%20Effect%20Class%20Reference))
  (bind (_outeffecthandle _shouldapply) (|CallParentFunction EffectHandle EffectClass))
  (bind _returnvalue (Ability|GameplayEffect|AddAssetTag _outeffecthandle (select (CallFunction|IsInstant) "(TagName=\"Asset.Creative.CreativeEffect.StatPowerup.Instant\")" "(TagName=\"Asset.Creative.CreativeEffect.StatPowerup.HasDuration\")")))
  (bind _returnvalue_2 (Ability|GameplayEffect|AddAssetTags _returnvalue (Utilities|Struct|BreakStatEventFilter (Variables|UserOptions|Basic|GetStatToApply))))
  (return _returnvalue_2 _shouldapply))

(fn UpdateStat
  (params (Stat%20Tag :type Gameplay%20Tag%20Structure) (IsValid :type Boolean) (Icon :type Texture%202D%20Soft%20Object%20Reference) (Color :type Linear%20Color%20Structure) (Stat%20name :type Text))
  ;[1 0 0.3311 0.25] Called from all Stat Creators when they update in the editor. This is due to the Editor not having the player controller component to broadcast stat detail changes.
  (if (GameplayTags|HasTag (Utilities|Struct|BreakStatEventFilter (Variables|UserOptions|Basic|GetStatToApply)) Stat%20Tag)
    (Variables|Default|SetStatColor Color)
    (Variables|Default|SetStatIcon Icon)
    (CallFunction|GetandUpdateStatDetails)))

(fn ConstructionScript
  (|CallParentFunction))
