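; Stat Counter Creative device that shows a tracked minigame stat with icon and number, broadcasts value-change and comparison events, and can override the stat.
; Stat details come from a playspace component and are cached.
; Delegates bound by server/client role and options refresh per-player cached values on stat, team, class and roster changes, comparing per player or in total.
; Comparison and override values replicate via RepNotify and can be set from Verse.
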(event OnAddedtoMinigame(FortMinigameProgress)
  (params (Minigame :type Fort%20Minigame%20Object%20Reference))
  ;[0 1 0.02482 0.204] Initialization
  (CallFunction|BindDelegates true))

(event EventInitializeDevice
  ;[0 1 0.02482 0.204] Initialization
  (|CallParentFunction)
  (CallFunction|InitializeComponents)
  (CallFunction|ResetCachedData)
  (CallFunction|CacheStatDetails)
  (CallFunction|BindDelegates true))

(event ChildFunctions|EventPlayModeChanged
  (params (Minigame :type Fort%20Minigame%20Object%20Reference) (bIsInPlayMode :type Boolean))
  ;[1 0.2608 0 0.607] Resetting
  (CallFunction|ResetCachedData)
  (CallFunction|UpdateVisibility))

(event OnEnabledStateChanged(EnabledComponent)
  (params (Enabled :type Boolean))
  ;[0 0.9901 1] Enable State Changes - Visibility updates based on enabled state, and when re-enabling we want to check stats.
  (CallFunction|UpdateVisibility)
  (CallFunction|BindDelegates Enabled)
  (if (and Enabled (Variables|UserOptions|Basic|GetBroadcast%20Events%20on%20Stat%20Change))
    (CallFunction|CheckAllPlayerStatsForChange false)))

(event EventEndPlay
  (params (EndPlayReason :type EEndPlayReason%20Enum :default "Destroyed"))
  ;[1 0 0 0.382] Unbind on EndPlay
  (CallFunction|BindDelegates false))

(event StatDetails|EventUpdateStat
  (params (Stat%20Tag :type Gameplay%20Tag%20Structure) (IsValid :type Boolean) (Icon :type Texture%202D%20Soft%20Object%20Reference) (Color :type Linear%20Color%20Structure) (Stat%20name :type Text))
  ;[0 1 0.5407 0.204] Update the static details of the device if the stat itself changes (e.g. someone alters the Stat Creator). This is a UEFN event that fires on all actors with the interface when a stat creator updates.
  (if (GameplayTags|HasTag (Utilities|Struct|BreakStatEventFilter (Variables|UserOptions|Basic|GetTrackedStat)) Stat%20Tag)
    (CallFunction|CacheStatDetails)))

(event Custom|OverrideValueEvent
  (params (Instigator :type Controller%20Object%20Reference))
  ;[0.1178 0 1 0.2] Gameplay Functions
  (Default|OverrideRelevantStatValue Instigator))

(event Custom|OnServerPlayerTeamChanged
  (params (PlayerState :type Fort%20Player%20State%20Object%20Reference) (TeamId :type Byte) (OldTeamId :type Byte) (bTeamChangedWithoutRespawn :type Boolean))
  (bind _returnvalue (Fortnite|GetControllerfromActor PlayerState))
  ;[0 1 0.9763 0.204] Update on Player State Change (switch team, class, enter or leave)
  (CallFunction|RemoveCacheIfInvalid _returnvalue)
  (CallFunction|UpdateonStatChanged _returnvalue))

(event Custom|EnableEvent
  (params (Instigator :type Controller%20Object%20Reference))
  ;[0.1178 0 1 0.2] Gameplay Functions
  (Class|CreativeEnabledComponent|SetEnabled (Variables|Default|GetEnabledComponent) true))

(event Custom|OnStatChanged
  (params (Type :type EFortQuestObjectiveStatEvent%20Enum :default "Kill") (TargetObject :type Object%20Reference) (Controller :type Controller%20Object%20Reference) (TargetTags :type Gameplay%20Tag%20Container%20Structure%20%28by%20ref%29) (SourceTags :type Gameplay%20Tag%20Container%20Structure%20%28by%20ref%29) (ContextTags :type Gameplay%20Tag%20Container%20Structure%20%28by%20ref%29) (Count :type Integer))
  ;[0 1 0.8836 0.204] Update on Any Stat Changed
  (CallFunction|UpdateonStatChanged Controller))

(event Custom|DisableEvent
  (params (Instigator :type Controller%20Object%20Reference))
  ;[0.1178 0 1 0.2] Gameplay Functions
  (Class|CreativeEnabledComponent|SetEnabled (Variables|Default|GetEnabledComponent) false))

(event Custom|Register%20Player%20Event
  (params (Instigator :type Controller%20Object%20Reference))
  ;[0.1178 0 1 0.2] Gameplay Functions
  (PlayerRegister|UpdatePlayerRegister Instigator true))

(event Custom|OnServerPlayerChangedClass
  (params (PlayerState :type Fort%20Player%20State%20Object%20Reference) (ClassSlotIndex :type Byte))
  (bind _returnvalue (Fortnite|GetControllerfromActor PlayerState))
  ;[0 1 0.9763 0.204] Update on Player State Change (switch team, class, enter or leave)
  (CallFunction|RemoveCacheIfInvalid _returnvalue)
  (CallFunction|UpdateonStatChanged _returnvalue))

(event Custom|OnClientStatChanged
  ;[0 1 0.8836 0.204] Update on Any Stat Changed
  (CallFunction|UpdateonStatChanged (Fortnite|GetFirstLocalFortPlayerController true) true))

(event Custom|Deregister%20Player%20Event
  (params (Instigator :type Controller%20Object%20Reference))
  ;[0.1178 0 1 0.2] Gameplay Functions
  (PlayerRegister|UpdatePlayerRegister Instigator))

(event Custom|OnPlayerPreRemoved
  (params (UniqueNetId :type Unique%20Net%20Id%20Structure) (bIsLocalPlayer :type Boolean))
  ;[0 1 0.9763 0.204] Update on Player State Change (switch team, class, enter or leave)
  (bind _returnvalue (Utilities|Map|Remove (Variables|Transient|GetCachedStatValue) (Fortnite|GetControllerfromActor (Game|GetPlayerStatefromUniqueNetId UniqueNetId))))
  (if _returnvalue
    (CallFunction|UpdateonStatChanged 0)))

(event Custom|Deregister%20All%20Event
  (params (Instigator :type Controller%20Object%20Reference))
  ;[0.1178 0 1 0.2] Gameplay Functions
  (PlayerRegister|ClearPlayerRegister))

(event Custom|OnServerPlayerJIP
  (params (Pawn :type Pawn%20Object%20Reference))
  ;[0 1 0.9763 0.204] Update on Player State Change (switch team, class, enter or leave)
  (CallFunction|UpdateonStatChanged (Fortnite|GetFortPlayerControllerfromActor Pawn)))

(event Custom|CompareStatEvent
  (params (Instigator :type Controller%20Object%20Reference))
  ;[0.1178 0 1 0.2] Gameplay Functions
  (CallFunction|UpdateonStatChanged Instigator true true))

(event Custom|Update%20Stat%20Details
  (bind _iconmid (Variables|Default|GetIconMID))
  ;[1 0.6099 0 0.2] Gathering the icon and color for the stat from the Player Controller Component (or the Device directly in editor) then loading the color)
  (Material|ReattachOrCreateMID (Variables|Default|GetIconPlane) _iconmid)
  (Material|ReattachOrCreateMID (Variables|Default|GetIconPlane_Back) _iconmid)
  (Rendering|Material|SetScalarParameterValue _iconmid "Master Opacity" (Math|Float|SelectFloat (Math|Float|SelectFloat 3.0 1.0 (== (Variables|CachedStatProperties|GetStatIcon) (Class|GetClassDefaults "/CRD_StatManager/Device_StatCounter.Device_StatCounter_C"))) 0.0 (Variables|UserOptions|Visuals|GetShowIcon)) :!class /Script/Engine.MaterialInstanceDynamic)
  (CallFunction|UpdateComponentLocationAndScale)
  (bind _staticon (Variables|CachedStatProperties|GetStatIcon)
    (:then
      (bind _loaded_object (Creative|Loading|AsyncLoadEditorSafe _staticon)
        (:Loaded
          (bind _astexture_2d (Utilities|Casting|CastToTexture2D _loaded_object))
          (Rendering|Material|SetTextureParameterValue _iconmid "Opacity Texture" _astexture_2d))))
    (:else
      (Rendering|Material|SetTextureParameterValue _iconmid "Opacity Texture" ""))))

(event Custom|OnStatsUpdated
  (params (New%20Details :type Map%20of%20Fort%20Minigame%20Stat%20Filter%20Object%20References%20to%20Stat%20Creator%20Details%20Structures))
  ;[1 0.6099 0 0.2] Gathering the icon and color for the stat from the Player Controller Component (or the Device directly in editor) then loading the color)
  (CallFunction|CacheStatDetails))

(event Custom|OnLocalPlayerChangedTeamOrClass
  ;[1 0.6099 0 0.2] Gathering the icon and color for the stat from the Player Controller Component (or the Device directly in editor) then loading the color)
  (CallFunction|UpdateVisibility))

(fn BindDelegates
  (params (Bind :type Boolean))
  (bind _minigame (LocalVariables|GetMinigame))
  (bind _returnvalue (Minigame|Stats|GetMinigameStatsComponent _minigame))
  (bind _minigame_2 (LocalVariables|GetMinigame))
  (bind _returnvalue_2 (Minigame|Stats|GetMinigameStatsComponent _minigame_2))
  (bind _compare_per_player (Variables|UserOptions|Comparison|GetCompare%20Per%20Player))
  (bind _returnvalue_3 (not _compare_per_player))
  (bind _bind (Variables|GetBind))
  (bind _broadcast_events_on_stat_change (Variables|UserOptions|Basic|GetBroadcast%20Events%20on%20Stat%20Change))
  (bind _returnvalue_4 (or _broadcast_events_on_stat_change _returnvalue_3))
  (bind _returnvalue_5 (and _returnvalue_4 _bind))
  (bind _outputdelegate (EventDispatchers|CreateEvent :!event OnClientStatChanged))
  (bind _outputdelegate_2 (EventDispatchers|CreateEvent :!event OnLocalPlayerChangedTeamOrClass))
  (bind _outputdelegate_3 (EventDispatchers|CreateEvent :!event OnLocalPlayerChangedTeamOrClass))
  (bind _outputdelegate_4 (EventDispatchers|CreateEvent :!event OnServerPlayerTeamChanged))
  (bind _outputdelegate_5 (EventDispatchers|CreateEvent :!event OnStatChanged))
  (bind _outputdelegate_6 (EventDispatchers|CreateEvent :!event OnServerPlayerChangedClass))
  (bind _outputdelegate_7 (EventDispatchers|CreateEvent :!event OnPlayerPreRemoved))
  (bind _outputdelegate_8 (EventDispatchers|CreateEvent :!event OnServerPlayerJIP))
  (bind _minigame_3 (LocalVariables|GetMinigame))
  (bind _minigame_4 (LocalVariables|GetMinigame))
  (bind _minigame_5 (LocalVariables|GetMinigame))
  (bind _minigame_6 (LocalVariables|GetMinigame))
  (bind _currentminigame (Variables|/Script/FortniteGame.FortMinigameLogicComponent|GetCurrentMinigame (Variables|Default|GetFortMinigameProgress))
    (:then
      (LocalVariables|SetMinigame _currentminigame)
      (Creative|Networking|Authority/Cosmetic
        (:Authority
          (LocalVariables|SetTrack%20Changes _returnvalue_5)
          ;[0 0.5928 1 0.2] Server Stat Track Delegate. For comparison checks on stat change.
          (if _returnvalue_5
            (Minigame|BindEventtoFullStatChangedDelegate _returnvalue_2 _outputdelegate_5)
            (Minigame|BindEventtoOnPlayerJoininProgress _minigame_2 _outputdelegate_8)
            (else
              (Minigame|UnbindEventfromFullStatChangedDelegate _returnvalue_2 _outputdelegate_5)
              (Minigame|UnbindEventfromOnPlayerJoininProgress _minigame_2 _outputdelegate_8)))
          (if (and (LocalVariables|GetTrack%20Changes) (!= (Creative|TeamOption|IsTeamType (Variables|UserOptions|Requirements|GetSelected%20Team)) (Variables|UserOptions|Requirements|GetInvert%20Team%20Selection)))
            (Minigame|BindEventtoPlayerChangedTeamDelegate _minigame_5 _outputdelegate_4)
            (else
              (Minigame|UnbindEventfromPlayerChangedTeamDelegate _minigame_5 _outputdelegate_4)))
          (if (Math|Boolean|ANDBoolean (LocalVariables|GetTrack%20Changes) (Variables|GetBind) (!= (Creative|ClassOption|IsClassType (Variables|UserOptions|Requirements|GetSelected%20Class)) (Variables|UserOptions|Requirements|GetInvert%20Class%20Selection)))
            (Minigame|BindEventtoOnPlayerClassApplied _minigame_6 _outputdelegate_6)
            (else
              (Minigame|UnbindEventfromOnPlayerClassApplied _minigame_6 _outputdelegate_6)))
          (if (not (Variables|UserOptions|Comparison|GetCompare%20Per%20Player))
            (Minigame|BindEventtoPlayerPreRemovedDelegate _minigame_3 _outputdelegate_7)
            (else
              (Minigame|UnbindEventfromPlayerPreRemovedDelegate _minigame_3 _outputdelegate_7))))
        (:Cosmetic
          (if (Math|Boolean|ANDBoolean (Variables|GetBind) (Utilities|Enum|NotEqual(Enum) (Variables|UserOptions|Visuals|GetValueToShow) "NewEnumerator2") (Utilities|Enum|NotEqual(Enum) (Variables|UserOptions|Visuals|GetValueToShow) "NewEnumerator3") (Utilities|Enum|NotEqual(Enum) (Variables|UserOptions|Visuals|GetVisibleInGame) "NewEnumerator2"))
            ;[1 0.5502 0 0.458] Client Stat Track Delegate. For Updating visual count.
            (Minigame|BindEventtoStatChangedDelegate _returnvalue _outputdelegate)
            (else
              (Minigame|UnbindEventfromStatChangedDelegate _returnvalue _outputdelegate)))
          (bind _asfort_player_state_athena (Utilities|Casting|CastToFortPlayerStateAthena (Variables|/Script/Engine.Controller|GetPlayerState (Fortnite|GetFirstLocalFortPlayerController true))))
          (if (Math|Boolean|ANDBoolean (Variables|GetBind) (Utilities|Enum|NotEqual(Enum) (Variables|UserOptions|Visuals|GetVisibleInGame) "NewEnumerator2") (!= (Creative|TeamOption|IsTeamType (Variables|UserOptions|Requirements|GetSelected%20Team)) (Variables|UserOptions|Requirements|GetInvert%20Team%20Selection)))
            (Player|BindEventtoOnTeamIndexChangedDelegate _asfort_player_state_athena _outputdelegate_3)
            (else
              (Player|UnbindEventfromOnTeamIndexChangedDelegate _asfort_player_state_athena _outputdelegate_3)))
          (if (Math|Boolean|ANDBoolean (Variables|GetBind) (Utilities|Enum|NotEqual(Enum) (Variables|UserOptions|Visuals|GetVisibleInGame) "NewEnumerator2") (!= (Creative|ClassOption|IsClassType (Variables|UserOptions|Requirements|GetSelected%20Class)) (Variables|UserOptions|Requirements|GetInvert%20Class%20Selection)))
            (Minigame|BindEventtoOnAnyLocalPlayerChangedClass _minigame_4 _outputdelegate_2)
            (else
              (Minigame|UnbindEventfromOnAnyLocalPlayerChangedClass _minigame_4 _outputdelegate_2))))))
    (:else)))

(fn BroadcastGameplayEvent
  (params (Event :type EStatCounterBroadcastEvent%20Enum :default "NewEnumerator0") (Instigator :type Controller%20Object%20Reference) (Value :type Integer))
  (bind _value (Variables|GetValue))
  (if (not (Variables|UserOptions|Comparison|GetCompare%20Per%20Player))
    (switch Utilities|FlowControl|Switch|SwitchonEStatCounterTransmitFor (Variables|UserOptions|Comparison|GetTransmit%20Events%20For)
      (:NewEnumerator1
        (Utilities|Array|Add (LocalVariables|GetTransmitPlayers) (Variables|Actor|GetInstigator)))
      (:NewEnumerator3
        (bind _controllers (CallFunction|GetAllMatchingPlayers))
        (Utilities|Array|Add (LocalVariables|GetTransmitPlayers) (Utilities|Array|RandomArrayItem _controllers)))
      (:NewEnumerator2
        (bind _controllers_2 (CallFunction|GetAllMatchingPlayers))
        (LocalVariables|SetTransmitPlayers _controllers_2)))
    (for _array_element (LocalVariables|GetTransmitPlayers)
      (switch Utilities|FlowControl|Switch|SwitchonEStatCounterBroadcastEvent (Variables|GetEvent)
        (:NewEnumerator0
          (GameplayEvent|BroadcastEvent (Variables|UserOptions|Event|GetOnValueChanged) 0 (Utilities|Struct|MakeGameplayEventStatManagerValueUp _array_element _value)))
        (:NewEnumerator1
          (GameplayEvent|BroadcastEvent (Variables|UserOptions|Event|GetOnCompareSuccess) 0 (Utilities|Struct|MakeGameplayEventStatCountSuccess _array_element _value)))
        (:NewEnumerator2
          (GameplayEvent|BroadcastEvent (Variables|UserOptions|Event|GetOnCompareFailure) 0 (Utilities|Struct|MakeGameplayEventStatCountFail _array_element (Variables|GetValue))))))))

(fn CacheStatDetails
  (bind (_staticon _statcolor) (Class|GetClassDefaults "/CRD_StatManager/Device_StatCounter.Device_StatCounter_C"))
  ;[0 0.774 1 0.2] Find the Stat Details from the Stat Playspace component and cache the values so we don't need to update again.
  (bind (_found _stat_icon_large_icon _statcolor_2 _valuefilter _scope) (Class|BPfFNEStats|GetStatDetails (Variables|UserOptions|Basic|GetTrackedStat) self))
  (bind _output_get (Variables|Default|SetFoundStat _found))
  (if _output_get
    (Variables|CachedStatProperties|SetStatIcon _stat_icon_large_icon)
    (Variables|CachedStatProperties|SetStatScope _scope)
    (Variables|CachedStatProperties|SetStatColor _statcolor_2)
    (Variables|CachedStatProperties|SetTrackedFilter _valuefilter)
    (else
      (Variables|CachedStatProperties|SetStatColor _statcolor)
      (Variables|CachedStatProperties|SetStatIcon _staticon)
      (Variables|CachedStatProperties|SetStatScope "NewEnumerator0")
      (Variables|CachedStatProperties|SetTrackedFilter 0)))
  (CallFunction|UpdateStatDetails)
  (Variables|Transient|SetIsTimeStat (Utilities|Array|ContainsItem (Variables|Default|GetTimeStats) (Variables|UserOptions|Basic|GetTrackedStat)))
  (CallFunction|UpdateStatVisuals 0))

(fn CheckAllPlayerStatsforChange
  (params (ForceCheck :type Boolean) (ForceUpdateCompare :type Boolean))
  (bind _currentminigame (Variables|/Script/FortniteGame.FortMinigameLogicComponent|GetCurrentMinigame (Variables|Default|GetFortMinigameProgress))
    (:then
      (bind _outplayers (Minigame|GetParticipatingPlayers_NotPure _currentminigame))
      (for _array_element _outplayers
        (CallFunction|UpdateonStatChanged (Fortnite|GetFortPlayerControllerfromActor _array_element) ForceCheck ForceUpdateCompare)))
    (:else)))

(fn ClearPlayerRegister
  (Variables|Transient|SetRegisteredPlayers (Utilities|Array|MakeArray)))

(fn FindOrCreatePlayspaceComponent
  ;[0 0.774 1 0.2] Generate Component if Necessary
  (Variables|Default|GetStatDetailsComponent
    (:then
      _)
    (:else
      (Networking|SwitchHasAuthority
        (:Authority
          (bind _returnvalue (Class|PlayspaceLibrary|GetPlayspaceforActor self self))
          (bind _output_get (Variables|Default|SetStatDetailsComponent (Actor|GetComponentbyClass _returnvalue "/CRD_StatManager/UI/PlayspaceComponent_PlayerStatDetails.PlayspaceComponent_PlayerStatDetails_C")))
          (Utilities|IsValid _output_get
            (:Is%20Valid
              _)
            (:Is%20Not%20Valid
              (Networking|SwitchHasAuthority
                (:Authority
                  (bind _returnvalue_2 (AddComponent|NODE_AddPlayspaceComponent_PlayerStatDetails-0 :self _returnvalue))
                  (Variables|Default|SetStatDetailsComponent _returnvalue_2)))))
          (return (Variables|Default|GetStatDetailsComponent)))))))

(fn GetAllMatchingPlayers
  (bind _currentminigame (Variables|/Script/FortniteGame.FortMinigameLogicComponent|GetCurrentMinigame (Variables|Default|GetFortMinigameProgress))
    (:then
      (bind _outplayers (Minigame|GetParticipatingPlayers_NotPure _currentminigame))
      (for _array_element _outplayers
        (Utilities|IsValid _array_element
          (:Is%20Valid
            (Default|ValidforPlayer (PlayerState|GetPlayerController _array_element)
              (:Yes
                (Utilities|Array|Add (LocalVariables|GetValidControllers) (PlayerState|GetPlayerController _array_element)))))))
      (return (LocalVariables|GetValidControllers)))
    (:else)))

(fn GetEnabledStateasIndex
  (return (Class|CreativeCommonDeviceFunctionLibrary2|EnabledPhaseEnumToLegacyInt (select (Variables|UserOptions|Basic|GetEnabled%20during%20Game) "NewEnumerator0" "NewEnumerator1"))))

(fn GetPlayerStatValue
  (params (For%20Controller :type Controller%20Object%20Reference))
  (bind _cachedminigamestatscomponent (Variables|Default|GetCachedMinigameStatsComponent))
  (bind _trackedfilter (Variables|CachedStatProperties|GetTrackedFilter))
  (bind _trackgametotal (Variables|Default|GetTrackGameTotal))
  ;[0 0.774 1 0.2] Returns the stat value for the specific player, gathered directly from the minigame stat list. This takes into account time stats, which it converts from milliseconds to seconds.
  (Utilities|GetMinigameStats
    (:Found
      (Utilities|GetTrackedFilter
        (:Valid
          (Utilities|IsValid For%20Controller
            (:Is%20Valid
              (bind _asfort_player_state (Utilities|Casting|CastToFortPlayerState (Variables|/Script/Engine.Controller|GetPlayerState For%20Controller)))
              (LocalVariables|SetDivisor (Math|Float|SelectFloat 1e+03 1.0 (Variables|Transient|GetIsTimeStat)))
              (switch Utilities|FlowControl|Switch|SwitchonE_StatManager_Scope (Variables|CachedStatProperties|GetStatScope)
                (:NewEnumerator0
                  (LocalVariables|SetValue (Minigame|Stats|GetStatforPlayer _cachedminigamestatscomponent _asfort_player_state _trackedfilter _trackgametotal)))
                (:NewEnumerator1
                  (LocalVariables|SetValue (Minigame|Stats|GetStatforPlayerBucket _cachedminigamestatscomponent (Teams|TeamtoIndex (Team|GetActorTeam For%20Controller)) _trackedfilter _trackgametotal false)))
                (:NewEnumerator2
                  (LocalVariables|SetValue (Minigame|Stats|GetStatforGroup _cachedminigamestatscomponent _trackedfilter _trackgametotal))))
              (return (Math|Float|Round (/ (Math|Conversions|ToFloat(Integer) (LocalVariables|GetValue)) (LocalVariables|GetDivisor)))))))))))

(fn GetPlayerStatValueConst
  (params (For%20Controller :type Controller%20Object%20Reference))
  (bind _cachedminigamestatscomponent (Variables|Default|GetCachedMinigameStatsComponent))
  (bind _trackedfilter (Variables|CachedStatProperties|GetTrackedFilter))
  (bind _trackgametotal (Variables|Default|GetTrackGameTotal))
  (Variables|Default|GetCachedMinigameStatsComponent
    (:then
      (Variables|CachedStatProperties|GetTrackedFilter
        (:then
          (Utilities|IsValid For%20Controller
            (:Is%20Valid
              (bind _asfort_player_state (Utilities|Casting|CastToFortPlayerState (Variables|/Script/Engine.Controller|GetPlayerState For%20Controller)))
              (switch Utilities|FlowControl|Switch|SwitchonE_StatManager_Scope (Variables|CachedStatProperties|GetStatScope)
                (:NewEnumerator0
                  (return (Minigame|Stats|GetStatforPlayer _cachedminigamestatscomponent _asfort_player_state _trackedfilter _trackgametotal)))
                (:NewEnumerator1
                  (return (Minigame|Stats|GetStatforPlayerBucket _cachedminigamestatscomponent (Teams|TeamtoIndex (Team|GetActorTeam (Variables|GetFor%20Controller))) _trackedfilter _trackgametotal false)))
                (:NewEnumerator2
                  _)))
            (:Is%20Not%20Valid
              (if (Utilities|Enum|Equal(Enum) (Variables|CachedStatProperties|GetStatScope) "NewEnumerator2")
                _)))
          (return (Minigame|Stats|GetStatforGroup _cachedminigamestatscomponent _trackedfilter _trackgametotal)))
        (:else)))
    (:else)))

(fn GetTextLengthforValue
  (params (InInt :type Integer))
  (if (Variables|UserOptions|Visuals|GetNumberLength_Override)
    (return (Variables|UserOptions|Visuals|GetNumberLength))
    (else
      (return 100))))

(fn GetComparisonValue
  ;[0.3436 0 1 0.2] Returns the current comparison value (This is called both from this device and from verse)
  (return (Variables|Default|GetComparisonValue_Runtime)))

(fn GetMaxValueForLength
  (params (A :type Integer))
  ;[1 1 1] Get the largest number that will fit in an integer with the input length. E.G. 5 length would be 99999. This equates to 1*10^length-1
  (return (Math|Integer|SelectInt (Math|Float|Truncate (- (Math|Float|Power 1e+01 (Math|Conversions|ToFloat(Integer) A)) 1.0)) 2147483647 (Variables|UserOptions|Visuals|GetNumberLength_Override))))

(fn GetTextForValue
  (params (Value :type Integer) (MaxValue :type Integer) (Digits :type Integer) (Return%20Time :type Boolean) (Force%20Sign :type Boolean))
  (if Return%20Time
    (bind _output_get (LocalVariables|SetMinutes (/ (Variables|GetValue) 60)))
    (bind _output_get_2 (LocalVariables|SetSeconds (Math|WrappingModulo (Variables|GetValue) 60)))
    (if (> (Variables|GetDigits) 2)
      (bind _output_get_3 (LocalVariables|SetText (Utilities|Text|FormatText "NSLOCTEXT(\"[90C3D4A0CB3DD456E74C1090E74C7254]\", \"BAF13CBC475E9EB330C8848FBAEF55C7\", \"{min}:{sec}\")" (Utilities|Text|ToText(Integer) _output_get (Variables|GetForce%20Sign) false (Math|Integer|SelectInt (- (Variables|GetDigits) 2) 1 (Variables|UserOptions|Visuals|GetNumberLength_Override)) (Math|Integer|SelectInt (Variables|GetDigits) 324 (Variables|UserOptions|Visuals|GetNumberLength_Override))) (Utilities|Text|ToText(Integer) _output_get_2 false false 2 2))))
      (if (and (> (Utilities|String|Len (Utilities|String|ToString(Text) _output_get_3)) (+ (Variables|GetDigits) 1)) (Variables|UserOptions|Visuals|GetNumberLength_Override))
        ;[1 0.586 0 0.2] If our number is too large for the display, just return the minutes.
        (bind _output_get_4 (LocalVariables|SetText (Utilities|Text|FormatText "NSLOCTEXT(\"[90C3D4A0CB3DD456E74C1090E74C7254]\", \"8F0AEAE64B917916161971A73D5A666B\", \"{min}m\")" (LocalVariables|GetMinutes))))
        (if (> (Utilities|String|Len (Utilities|String|ToString(Text) _output_get_4)) (Variables|GetDigits))
          (return (Utilities|Text|FormatText "NSLOCTEXT(\"[90C3D4A0CB3DD456E74C1090E74C7254]\", \"9F9037C14AA2C3CF6FD45C8CC052178E\", \"{sign}{min}m\")" (Utilities|String|SelectString "" "-" (Math|Conversions|ToBoolean(Integer) (Math|Integer|Sign(Integer) (LocalVariables|GetMinutes)))) (Utilities|String|Left "9999999999" (- (Variables|GetDigits) (Math|Integer|SelectInt 1 2 (Math|Conversions|ToBoolean(Integer) (Math|Integer|Sign(Integer) (LocalVariables|GetMinutes))))))))
          (else
            (return (LocalVariables|GetText))))
        (else
          (return (LocalVariables|GetText))))
      (else
        ;[1 0.803 0 0.1] Only one digit, so severely limited in what we show. Only show the minute or  number or the second number, maxed to the limited number of digits
        (return (Utilities|Text|ToText(Integer) (Math|Integer|SelectInt (Math|Integer|Clamp(Integer) (LocalVariables|GetMinutes) 0 9) (Math|Integer|Clamp(Integer) (LocalVariables|GetSeconds) 0 9) (> (Math|Integer|Clamp(Integer) (LocalVariables|GetMinutes) 0 9) 0)) false false (Variables|GetDigits) (Variables|GetDigits)))))
    (else
      (return (Utilities|Text|ToText(Integer) (Math|Integer|Clamp(Integer) (Variables|GetValue) (* MaxValue -1) MaxValue) (Variables|GetForce%20Sign) false (Math|Integer|SelectInt Digits (Variables|Default|GetDefault%20Minimum%20Digits) (Variables|UserOptions|Visuals|GetNumberLength_Override)) Digits)))))

(fn GetTrackedValue
  (params (Controller :type Controller%20Object%20Reference))
  ;[0.3436 0 1 0.2] Returns the total value for the tracked players.
  (if (and (Variables|UserOptions|Comparison|GetCompare%20Per%20Player) (Utilities|Enum|NotEqual(Enum) (Variables|CachedStatProperties|GetStatScope) "NewEnumerator2"))
    (Default|ValidforPlayer Controller
      (:Yes
        (bind _returnvalue (CallFunction|GetPlayerStatValueConst Controller))
        (return _returnvalue))
      (:No
        (return 0)))
    (else
      (Utilities|FlowControl|Sequence
        (:then_0
          (switch Utilities|FlowControl|Switch|SwitchonE_StatManager_Scope (Variables|CachedStatProperties|GetStatScope)
            (:NewEnumerator0
              ;;[0 0.774 1 0.2] Gather values of all players that have cached values, since we are tracking per player.
              (bind _values (Utilities|Map|Values (Variables|Transient|GetCachedStatValue)))
              (for _array_element _values
                (LocalVariables|SetSum (+ _array_element (LocalVariables|GetSum)))))
            (:NewEnumerator1
              (bind _keys (Utilities|Map|Keys (Variables|Transient|GetCachedStatValue)))
              (for _array_element_2 _keys
                (bind _returnvalue_2 (CallFunction|GetPlayerStatValueConst _array_element_2))
                (if (not (Utilities|Set|ContainsItem (LocalVariables|GetAccumulatedTeams) (Team|GetActorTeam _array_element_2)))
                  (LocalVariables|SetSum (+ _returnvalue_2 (LocalVariables|GetSum)))
                  (Utilities|Set|Add (LocalVariables|GetAccumulatedTeams) (Team|GetActorTeam _array_element_2)))))
            (:NewEnumerator2
              (bind _returnvalue_3 (CallFunction|GetPlayerStatValueConst 0)) ; Just get the value for the match.\nDoesn't need a controller.
              (LocalVariables|SetSum _returnvalue_3))))
        (:then_1
          (return (LocalVariables|GetSum)))))))

(fn InitializeComponents
  (Initialization|Initialize (Variables|Default|GetCreative_UseRestriction_Component) (Variables|Default|GetFortMinigameProgress) (Variables|Default|GetEnabledComponent) (Variables|Default|GetToyOptionsComponent) (Variables|UserOptions|Requirements|GetSelected%20Team) (Variables|UserOptions|Requirements|GetInvert%20Team%20Selection) (Variables|UserOptions|Requirements|GetSelected%20Class) (Variables|UserOptions|Requirements|GetInvert%20Class%20Selection) false))

(fn OnRep_CachedGlobalValue
  (Networking|CanExecuteCosmeticEvents
    (:True
      (if (not (Variables|UserOptions|Comparison|GetCompare%20Per%20Player))
        (CallFunction|UpdateStatVisuals (Variables|Transient|GetCached%20Global%20Value))))))

(fn OnRep_ComparisonValue_Runtime
  ;[0 0.5225 1 0.266] Our Comparison Value has changed (Verse updated it). If we're the server, recheck all players (if we want to compare on stat change). If we're the clien, update the visuals.
  (Creative|Networking|Authority/Cosmetic
    (:Authority
      (if (and (Class|CreativeEnabledComponent|IsEnabled (Variables|Default|GetEnabledComponent)) (Variables|UserOptions|Basic|GetBroadcast%20Events%20on%20Stat%20Change))
        (CallFunction|CheckAllPlayerStatsForChange false true)))
    (:Cosmetic
      (if (not (Variables|UserOptions|Comparison|GetCompare%20Per%20Player))
        (CallFunction|UpdateStatVisuals (Variables|Transient|GetCached%20Global%20Value))
        (else
          (bind _returnvalue (CallFunction|GetPlayerStatValue (Fortnite|GetFirstLocalFortPlayerController true)))
          (CallFunction|UpdateStatVisuals _returnvalue))))))

(fn OnRep_OverrideValue_Runtime
  (Networking|CanExecuteCosmeticEvents
    (:True
      (if (Utilities|Enum|Equal(Enum) (Variables|UserOptions|Visuals|GetValueToShow) "NewEnumerator3")
        (CallFunction|UpdateStatVisuals 0)))))

(fn OnRep_RegisteredPlayers
  (CallFunction|UpdateVisibility))

(fn OnRep_StatDetailsComponent
  (bind _statdetailscomponent (Variables|Default|GetStatDetailsComponent)
    (:then
      (Default|BindEventtoOnClientDetailsUpdated _statdetailscomponent (EventDispatchers|CreateEvent :!event OnStatsUpdated))
      (CallFunction|OnStatsUpdated (Variables|/CRD_StatManager/UI/PlayspaceComponent_PlayerStatDetails.PlayspaceComponent_PlayerStatDetails_C|GetStat%20Creator%20Details _statdetailscomponent)))
    (:else)))

(fn OverrideRelevantStatValue
  (params (PlayerInstigator :type Controller%20Object%20Reference) (UseInputOverride :type Boolean) (InputOverride :type Integer))
  (bind (_statevent _targettagsquery) (Utilities|Struct|BreakStatEventFilter (Variables|UserOptions|Basic|GetTrackedStat)))
  (bind _useinputoverride (Variables|GetUseInputOverride))
  (bind _overridevalue_runtime (Variables|Default|GetOverrideValue_Runtime))
  (bind _inputoverride (Variables|GetInputOverride))
  (bind _returnvalue (Math|Integer|SelectInt _inputoverride _overridevalue_runtime _useinputoverride))
  (bind _multiplier (LocalVariables|GetMultiplier))
  (bind _returnvalue_2 (* _returnvalue _multiplier))
  (bind _returnvalue_3 (Math|Float|Round _returnvalue_2))
  (Networking|SwitchHasAuthority
    (:Authority
      (Default|ValidforPlayer PlayerInstigator
        (:Yes
          (Utilities|FlowControl|Sequence
            (:then_0
              ;[0 0.774 1 0.2] Gather the value we want to change our stat by (since we add stats)
              (LocalVariables|SetMultiplier (Math|Float|SelectFloat 1e+03 1.0 (Variables|Transient|GetIsTimeStat)))
              (switch Utilities|FlowControl|Switch|SwitchonEAwardScoreType (Variables|UserOptions|Override|GetValue%20Override%20Type)
                (:NewEnumerator1
                  (LocalVariables|SetValueToSet _returnvalue_3))
                (:NewEnumerator2
                  (LocalVariables|SetValueToSet (* _returnvalue_3 -1)))
                (:NewEnumerator3
                  (bind _returnvalue_4 (CallFunction|GetPlayerStatValue (Variables|GetPlayerInstigator)))
                  (LocalVariables|SetValueToSet (- _returnvalue_3 (* _returnvalue_4 (LocalVariables|GetMultiplier)))))))
            (:then_1
              ;[0.4991 1 0.7068 0.204] Ensure MinigameStats and TrackedFilter
              ; are populated.
              (Utilities|GetMinigameStats
                (:Found
                  (Utilities|GetTrackedFilter
                    (:Valid
                      (if (Class|CreativeEnabledComponent|IsEnabled (Variables|Default|GetEnabledComponent))
                        (bind _returnvalue_5 (CreativeDevices|GetMinigameStatsComp (Variables|GetPlayerInstigator) "Athena"))
                        (Utilities|IsValid _returnvalue_5
                          (:Is%20Valid
                            (Class|FortMinigameStatsComponent|SendStatEventwithTags _returnvalue_5 (Variables|GetPlayerInstigator) _statevent self _targettagsquery (LocalVariables|GetEmpty) (LocalVariables|GetEmpty) (LocalVariables|GetValueToSet)) ; Stats are handled with SendStatEventWithTags.\nThis will trigger the actual StatCreator to handle the \nstat addition (so it can handle max/min values and levels)
                          ))))))))))))))

(fn RemoveCacheifInvalid
  (params (Controller :type Controller%20Object%20Reference%20%28by%20ref%29))
  (Default|ValidforPlayer Controller
    (:No
      (Utilities|Map|Remove (Variables|Transient|GetCachedStatValue) Controller))))

(fn ResetCachedData
  (Utilities|Map|Clear (Variables|Transient|GetComparisonMap))
  (Utilities|Map|Clear (Variables|Transient|GetCachedStatValue))
  (Variables|Transient|SetCached%20Global%20Value 0)
  (PlayerRegister|ClearPlayerRegister)
  (Variables|Transient|SetGlobalHasReached false)
  (Creative|Networking|Authority/Cosmetic
    (:Authority
      (if (!= (Variables|UserOptions|Comparison|GetComparisonValue) (Verse|GetComparisonValue))
        (Verse|SetComparisonValue (Variables|UserOptions|Comparison|GetComparisonValue))
        (else
          _))
      (if (!= (Variables|UserOptions|Override|GetValue%20Override) (Variables|Default|GetOverrideValue_Runtime))
        (Verse|SetOverrideValue (Variables|UserOptions|Override|GetValue%20Override))))
    (:Cosmetic
      (CallFunction|UpdateStatVisuals 0))))

(fn SetComparisonValue
  (params (NewValue :type Integer))
  ;[0.3436 0 1 0.2] Update the Comparison Value from Verse. This will RepNotify the value, which will SERVER: Recalculate comparison if necessary, CLIENT: Update visuals.
  (if (!= NewValue (Variables|Default|GetComparisonValue_Runtime))
    (Variables|Default|SetComparisonValue_Runtime NewValue)
    (return)))

(fn SetOverrideValue
  (params (NewValue :type Integer))
  ;[0.3436 0 1 0.2] Update the Override value if it's changed.
  (if (!= NewValue (Variables|Default|GetOverrideValue_Runtime))
    (Variables|Default|SetOverrideValue_Runtime NewValue)
    (return)))

(fn SetPlayerValue
  (params (Agent :type Controller%20Object%20Reference) (Int :type Integer))
  (Default|OverrideRelevantStatValue Agent true Int))

(fn UpdateComponentLocationandScale
  (bind _currentstat (Variables|Default|GetCurrentStat))
  (bind (_returnvalue_x _returnvalue_y) (Math|Vector|BreakVector (Rendering|Components|TextRender|GetTextLocalSize _currentstat)))
  (bind _numberlength_override (Variables|UserOptions|Visuals|GetNumberLength_Override))
  (bind _length (LocalVariables|GetLength))
  (bind _backgroundlengthpernumber (Variables|Resizing|GetBackgroundLengthPerNumber))
  (bind _returnvalue (* _backgroundlengthpernumber _length))
  (bind _returnvalue_2 (Math|Float|SelectFloat _returnvalue _returnvalue_y _numberlength_override))
  (bind _backgroundpadding (Variables|Resizing|GetBackgroundPadding))
  (bind _returnvalue_3 (* _backgroundpadding 2.0))
  (bind _numberlength (Variables|UserOptions|Visuals|GetNumberLength))
  (bind _returnvalue_4 (> _numberlength 0))
  (bind _icon_size (Variables|Resizing|GetIcon%20Size))
  (bind _returnvalue_5 (/ _icon_size -2.0))
  (bind _showicon (Variables|UserOptions|Visuals|GetShowIcon))
  (bind _icon_size_2 (Variables|Resizing|GetIcon%20Size))
  (bind _returnvalue_6 (select _showicon _icon_size_2 0.0))
  (bind _returnvalue_7 (Utilities|Operators|Add _returnvalue_2 _returnvalue_6 _returnvalue_3))
  (Rendering|SetVisibility (Variables|Default|GetCurrentStat) _returnvalue_4 false)
  (if (or (not (Variables|UserOptions|Visuals|GetNumberLength_Override)) _returnvalue_4)
    (LocalVariables|SetLength (Utilities|Select :Index (Variables|UserOptions|Visuals|GetValueToShow) :NewEnumerator0 (* (Variables|UserOptions|Visuals|GetNumberLength) 2) :NewEnumerator1 (Variables|UserOptions|Visuals|GetNumberLength) :NewEnumerator2 (Variables|UserOptions|Visuals|GetNumberLength) :NewEnumerator3 (+ (Variables|UserOptions|Visuals|GetNumberLength) (Utilities|Enum|NotEqual(Enum) (Variables|UserOptions|Override|GetValue%20Override%20Type) "NewEnumerator3"))))
    (Rendering|SetVisibility (Variables|Default|GetStatCounter_Background) (Variables|UserOptions|Visuals|GetShowBackground) false)
    (Transformation|SetRelativeLocation (Variables|Default|GetFrontDetails) (Math|Vector|MakeVector -0.2 (/ _returnvalue_7 -2.0)) false false)
    (Transformation|SetRelativeLocation (Variables|Default|GetBackDetails) (Math|Vector|MakeVector 0.2 (/ _returnvalue_7 2.0)) false false)
    (Transformation|SetRelativeLocation (Variables|Default|GetCurrentStat) (Math|Vector|MakeVector 0.0 (+ (select (Variables|UserOptions|Visuals|GetShowIcon) (Variables|Resizing|GetIcon%20Size) 0.0) (Variables|Resizing|GetBackgroundPadding))) false false)
    (Transformation|SetRelativeScale3D (Variables|Default|GetStatCounter_Background) (Math|Vector|MakeVector _returnvalue_7 1.0 1.0))
    (else
      ;[1 1 1] No numbers, just the icon (if even that). Statically place the objects.
      (Transformation|SetRelativeScale3D (Variables|Default|GetStatCounter_Background) (Math|Vector|MakeVector (Variables|Resizing|GetIcon%20Size) 1.0 1.0))
      (Transformation|SetRelativeLocation (Variables|Default|GetFrontDetails) (Math|Vector|MakeVector -0.2 _returnvalue_5) false false)
      (Transformation|SetRelativeLocation (Variables|Default|GetBackDetails) (Math|Vector|MakeVector 0.2 (* _returnvalue_5 -1.0)) false false)
      (Rendering|SetVisibility (Variables|Default|GetStatCounter_Background) (or (Variables|UserOptions|Visuals|GetShowBackground) (not (Variables|UserOptions|Visuals|GetShowIcon))) false) ; If we're not showing icon, and we're not showing\nnumbers, force the background on so there's something for Creators to see. 
    )))

(fn UpdatePlayerRegister
  (params (Controller :type Controller%20Object%20Reference%20%28by%20ref%29) (Add :type Boolean))
  (bind _localarray (LocalVariables|GetLocalArray))
  ;[0.2149 0.7535 1 0.295] Adds or removes from the player register, and then re-evaluates the player if the register changed.
  (LocalVariables|SetLocalArray (Variables|Transient|GetRegisteredPlayers))
  (if Add
    (bind _returnvalue (Utilities|Array|AddUnique _localarray Controller))
    (if (> _returnvalue -1)
      (Variables|Transient|SetRegisteredPlayers (LocalVariables|GetLocalArray))
      (return true)
      (else
        _))
    (else
      (bind _returnvalue_2 (Utilities|Array|RemoveItem _localarray Controller))
      (if _returnvalue_2
        _
        (else
          _))))
  (return false))

(fn UpdateStat
  (params (TestController :type Controller%20Object%20Reference))
  (bind (_value _returnvalue) (Utilities|Map|Find (Variables|Transient|GetCachedStatValue) (Variables|GetTestController)))
  (if (Class|CreativeEnabledComponent|IsEnabled (Variables|Default|GetEnabledComponent))
    (Default|ValidforPlayer TestController
      (:Yes
        (bind _returnvalue_2 (CallFunction|GetPlayerStatValue TestController))
        (LocalVariables|SetValue _returnvalue_2)
        (bind _output_get (LocalVariables|SetChanged (or (!= (LocalVariables|GetValue) _value) (not _returnvalue))))
        (if _output_get
          (Utilities|Map|Add (Variables|Transient|GetCachedStatValue) (Variables|GetTestController) (LocalVariables|GetValue))
          (else
            (return false (LocalVariables|GetValue)))))
      (:No
        _))
    (Utilities|FlowControl|Sequence
      (:then_0
        ;[0 1 0.5407 0.204] If we're not tracking the player, we want to gather all of the cached values and update if they change.
        (if (not (Variables|UserOptions|Comparison|GetCompare%20Per%20Player))
          (bind _count (Verse|GetTrackedValue 0))
          (LocalVariables|SetValue _count)
          (bind _output_get_2 (LocalVariables|SetChanged (!= (LocalVariables|GetValue) (Variables|Transient|GetCached%20Global%20Value))))
          (if _output_get_2
            (Variables|Transient|SetCached%20Global%20Value (LocalVariables|GetValue))
            (if (Utilities|Enum|NotEqual(Enum) (Variables|UserOptions|Visuals|GetVisibleInGame) "NewEnumerator2")
              (Networking|ForceNetUpdate)))))
      (:then_1
        (return (LocalVariables|GetChanged) (LocalVariables|GetValue))))
    (else
      (return false 0))))

(fn UpdateStatVisuals
  (params (New%20Stat%20Value :type Integer))
  (bind _istimestat (Variables|Transient|GetIsTimeStat))
  (bind _textlengthtarget (LocalVariables|GetTextLengthTarget))
  (bind _maxtarget (LocalVariables|GetMaxTarget))
  (bind _count (Verse|GetComparisonValue))
  (bind _returnvalue (CallFunction|GetTextforValue _count _maxtarget _textlengthtarget _istimestat))
  (bind _istimestat_2 (Variables|Transient|GetIsTimeStat))
  (bind _textlengthvalue (LocalVariables|GetTextLengthValue))
  (bind _maxvalue (LocalVariables|GetMaxValue))
  (bind _returnvalue_2 (CallFunction|GetTextforValue New%20Stat%20Value _maxvalue _textlengthvalue _istimestat_2))
  ;[1 0.6509 0 0.25] Updates the numerical text for the device based on the settings you have chosen. This takes into account:
  ; Length of text depending on if you have an icon displayed or not, as to not overrun the device mesh.
  ; Addition/Subtraction if you are set to display the override value.
  ; Showing time, if you are set to a stat that's time based.
  (Networking|CanExecuteCosmeticEvents
    (:True
      (Utilities|FlowControl|Sequence
        (:then_0
          (if (and (not (Math|Conversions|ToBoolean(Integer) (Variables|UserOptions|Visuals|GetNumberLength))) (Variables|UserOptions|Visuals|GetNumberLength_Override))
            (Rendering|Components|TextRender|SetText (Variables|Default|GetCurrentStat) (Utilities|Text|MakeLiteralText ""))
            (else
              (bind _output_get (LocalVariables|SetTextLengthValue (CallFunction|GetTextLengthForValue New%20Stat%20Value)))
              (bind _output_get_2 (LocalVariables|SetTextLengthTarget (CallFunction|GetTextLengthForValue (Verse|GetComparisonValue))))
              (LocalVariables|SetMaxValue (CallFunction|GetMaxValueforLength _output_get))
              (LocalVariables|SetMaxTarget (CallFunction|GetMaxValueforLength _output_get_2))
              (Rendering|Components|TextRender|SetTextRenderColor (Variables|Default|GetCurrentStat) (Math|Conversions|ToColor(LinearColor) (Variables|CachedStatProperties|GetStatColor) false))
              (if (Variables|Default|GetFoundStat)
                (switch Utilities|FlowControl|Switch|SwitchonEStatManagerVisualsInGame (Variables|UserOptions|Visuals|GetValueToShow)
                  (:NewEnumerator0
                    (Rendering|Components|TextRender|SetText (Variables|Default|GetCurrentStat) (Utilities|Text|FormatText "NSLOCTEXT(\"[90C3D4A0CB3DD456E74C1090E74C7254]\", \"A199F058414E2286909DBFB177D920DF\", \"{value}/{target}\")" _returnvalue_2 _returnvalue)))
                  (:NewEnumerator1
                    (Rendering|Components|TextRender|SetText (Variables|Default|GetCurrentStat) _returnvalue_2))
                  (:NewEnumerator2
                    (Rendering|Components|TextRender|SetText (Variables|Default|GetCurrentStat) _returnvalue))
                  (:NewEnumerator3
                    (Rendering|Components|TextRender|SetText (Variables|Default|GetCurrentStat_Back) (CallFunction|GetTextforValue (* (Variables|Default|GetOverrideValue_Runtime) (Math|Integer|SelectInt -1 1 (Utilities|Enum|Equal(Enum) (Variables|UserOptions|Override|GetValue%20Override%20Type) "NewEnumerator2"))) (LocalVariables|GetMaxValue) (LocalVariables|GetTextLengthValue) (Variables|Transient|GetIsTimeStat) (and (or (Utilities|Enum|Equal(Enum) (Variables|UserOptions|Override|GetValue%20Override%20Type) "NewEnumerator2") (Utilities|Enum|Equal(Enum) (Variables|UserOptions|Override|GetValue%20Override%20Type) "NewEnumerator1")) (!= (Variables|Default|GetOverrideValue_Runtime) 0))))))
                (else
                  (Rendering|Components|TextRender|SetText (Variables|Default|GetCurrentStat) (Utilities|Text|ToText(String) (Utilities|String|LeftChop (Variables|Default|GetNoStatText) (Math|Integer|SelectInt 3 0 (Variables|UserOptions|Visuals|GetShowIcon))))))))))
        (:then_1
          (if (not (Variables|UserOptions|Visuals|GetNumberLength_Override))
            (CallFunction|UpdateComponentLocationAndScale) ; If we're not using a fixed length, update the background component length after the size.
          ))))))

(fn UpdateVisibility
  (bind _bisinplaymode (CallFunction|IsinPlayMode))
  (bind _returnvalue (Editor|InEditor))
  (bind _returnvalue_2 (or _returnvalue _bisinplaymode))
  (bind _shouldhide (LocalVariables|GetShouldHide))
  (Networking|CanExecuteCosmeticEvents
    (:True
      (Utilities|FlowControl|Sequence
        (:then_0
          (Default|ValidforPlayer (Fortnite|GetFirstLocalFortPlayerController true)
            (:Yes
              (LocalVariables|SetShouldHide (and (Utilities|Enum|Equal(Enum) (Variables|UserOptions|Visuals|GetVisibleInGame) "NewEnumerator2") _returnvalue_2))
              (LocalVariables|SetValid true))
            (:No
              (LocalVariables|SetShouldHide (and _returnvalue_2 (Utilities|Enum|NotEqual(Enum) (Variables|UserOptions|Visuals|GetVisibleInGame) "NewEnumerator0")))
              (LocalVariables|SetValid false))))
        (:then_1
          (Rendering|SetActorHiddenInGame _shouldhide)
          (if (not _shouldhide)
            (if (or (not (Variables|UserOptions|Comparison|GetCompare%20Per%20Player)) (LocalVariables|GetValid))
              (bind _returnvalue_3 (CallFunction|GetPlayerStatValue (Fortnite|GetFirstLocalFortPlayerController true)))
              (CallFunction|UpdateStatVisuals _returnvalue_3)
              (else
                (CallFunction|UpdateStatVisuals 0)))))))))

(fn UpdateComparison
  (params (Controller :type Controller%20Object%20Reference) (NewStatValue :type Integer) (ForceCheck :type Boolean))
  (bind _newreached (LocalVariables|GetNewReached))
  (bind _controller (Variables|GetController))
  (bind _newstatvalue (Variables|GetNewStatValue))
  (if (Variables|UserOptions|Comparison|GetCompare%20Per%20Player)
    (LocalVariables|SetHasReached (Utilities|Map|Find (Variables|Transient|GetComparisonMap) Controller))
    (else
      (LocalVariables|SetHasReached (Variables|Transient|GetGlobalHasReached))))
  (LocalVariables|SetNewReached (Utilities|Select :Index (Variables|UserOptions|Comparison|GetCompareType) :NewEnumerator0 false :NewEnumerator3 (< NewStatValue (Verse|GetComparisonValue)) :NewEnumerator4 (<= NewStatValue (Verse|GetComparisonValue)) :NewEnumerator6 (!= NewStatValue (Verse|GetComparisonValue)) :NewEnumerator1 (not (!= NewStatValue (Verse|GetComparisonValue))) :NewEnumerator5 (not (< NewStatValue (Verse|GetComparisonValue))) :NewEnumerator2 (not (<= NewStatValue (Verse|GetComparisonValue)))))
  (if (LocalVariables|GetNewReached)
    (if (or (not (LocalVariables|GetHasReached)) (Variables|GetForceCheck))
      (CallFunction|BroadcastGameplayEvent "NewEnumerator1" _controller _newstatvalue))
    (elif (or (LocalVariables|GetHasReached) (Variables|GetForceCheck))
      (CallFunction|BroadcastGameplayEvent "NewEnumerator2" _controller _newstatvalue)))
  (if (Variables|UserOptions|Comparison|GetCompare%20Per%20Player)
    (Utilities|Map|Add (Variables|Transient|GetComparisonMap) (Variables|GetController) _newreached)
    (else
      (Variables|Transient|SetGlobalHasReached _newreached))))

(fn UpdateOnStatChanged
  (params (Controller :type Controller%20Object%20Reference) (ForceCheck :type Boolean) (ForceUpdateCompare :type Boolean))
  (if (or ForceCheck (Variables|UserOptions|Basic|GetBroadcast%20Events%20on%20Stat%20Change))
    (bind (_haschanged _stat_value) (CallFunction|UpdateStat Controller))
    (if _haschanged
      (Creative|Networking|Authority/Cosmetic
        (:Authority
          (CallFunction|BroadcastGameplayEvent "NewEnumerator0" Controller _stat_value))
        (:Cosmetic
          (CallFunction|UpdateStatVisuals _stat_value)))
      (elif (Variables|GetForceUpdateCompare)
        (Creative|Networking|Authority/Cosmetic
          (:Authority
            _))))
    (CallFunction|UpdateComparison Controller _stat_value (Variables|GetForceCheck))))

(fn ConstructionScript
  (|CallParentFunction))
