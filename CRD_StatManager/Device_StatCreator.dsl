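; Stat Creator Creative device that defines a custom minigame stat with optional levels, scoped per player, team or match.
; Intercepts stat-add delegates and channel events to clamp values, derive levels from points-per-level tables and broadcast change events.
; Caches and zeroes stats on disable, restores them on enable, and revalidates players on team or class change.
; Persistence handles save, auto-load and join-in-progress, while a playspace component receives UI details.
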(event EventBeginPlay
  (|CallParentFunction)
  (Initialization|InitializePersistence (Variables|Default|GetFortMinigameProgress)))

(event EventOnMingameStateChanged
  (params (Minigame :type Fort%20Minigame%20Object%20Reference) (MinigameState :type EFortMinigameState%20Enum :default "PreGame"))
  ;[0 0.787 1 0.2] Minigame State Changed - Handle adding/removing stats and loading data.
  (|CallParentFunction Minigame MinigameState) ; Ensure parent logic (e.g. enabled state \nswitch) happens first, needed for Auto-Load/Save\nfunctionality to run.
  (Networking|SwitchHasAuthority
    (:Authority
      (switch Utilities|FlowControl|Switch|SwitchonEFortMinigameState MinigameState
        (:PostGameReset
          (Cache|ClearCaches))))))

(event EventInitializeDevice
  ;[0.04337 1 0.127 0.249] Initialization
  (|CallParentFunction)
  (if (not (Building|IsPreviewBuildingActor))
    (Initialization|Initialize (Variables|Default|GetCreative_UseRestriction_Component) (Variables|Default|GetFortMinigameProgress) (Variables|Default|GetEnabledComponent) (Variables|Default|GetToyOptionsComponent) (Variables|UserOptions|Requirements|GetSelected%20Team) (Variables|UserOptions|Requirements|GetInvert%20Team%20Selection) (Variables|UserOptions|Requirements|GetSelected%20Class) (Variables|UserOptions|Requirements|GetInvert%20Class%20Selection) false)
    (Creative|Networking|Authority/Cosmetic
      (:Authority
        (Initialization|CalculateExpTables)
        (Initialization|ValidateStatName)
        (Game|BindEventtoOnDestroyed (EventDispatchers|CreateEvent :!event Custom|OnDestroyed_Event))
        (CallFunction|SetupMinigameStat true) ; Setting up again during Minigame State Changed\nfor any stats that couldn't create at Init due to no slots available, but\nthe slot was later cleared by deleting of another stat.
        (UI|UpdateVisualDetailsFromServer))
      (:Cosmetic
        (CallFunction|UpdateDeviceIcon)))))

(event MinigameStartedDelegate(FortMinigameProgress)
  ;[0 0.787 1 0.2] Minigame State Changed - Handle adding/removing stats and loading data.
  (Networking|SwitchHasAuthority
    (:Authority
      (Utilities|FlowControl|Sequence
        (:then_0
          (if (and (Variables|UserOptions|Persistence|GetbAutoLoad) (Variables|UserOptions|Persistence|GetUsePersistence))
            (Persistence|LoadStatAll 0) ; Load Stats for all players
            (else
              _))
          (CallFunction|InitializeLevelForAllPlayers) ; Initializes Level, ensuring everyone relevant starts at \nLevel 1. Done after loading stats to ensure we don't nuke their level.
        )
        (:then_1
          (Persistence|SetAutoSaveEnabled (Variables|Default|GetFortCreativePersistenceComponentBase) (and (Variables|UserOptions|Persistence|GetbAutoSave) (Variables|UserOptions|Persistence|GetUsePersistence))))))))

(event OnAddedtoMinigame(FortMinigameProgress)
  (params (Minigame :type Fort%20Minigame%20Object%20Reference))
  ;[0.04337 1 0.127 0.249] Initialization
  (if (not (Building|IsPreviewBuildingActor))
    (Initialization|ValidateStatName)
    (CallFunction|BindMinigame true)
    (CallFunction|SetupMinigameStat true) ; Setting up again during Minigame State Changed\nfor any stats that couldn't create at Init due to no slots available, but\nthe slot was later cleared by deleting of another stat.
    (UI|UpdateVisualDetailsFromServer)))

(event MinigameEndedDelegate(FortMinigameProgress)
  ;[0 0.787 1 0.2] Minigame State Changed - Handle adding/removing stats and loading data.
  (Networking|SwitchHasAuthority
    (:Authority
      (Persistence|SetAutoSaveEnabled (Variables|Default|GetFortCreativePersistenceComponentBase) false))))

(event Custom|EnableEvent
  (params (Instigator :type Controller%20Object%20Reference))
  ;[0.4835 0 1 0.2] Channel Receivers
  (Class|CreativeEnabledComponent|SetEnabled (Variables|Default|GetEnabledComponent) true))

(event Custom|On%20Stat%20Value%20Changed
  (params (Amount :type Integer) (InstigatingController :type Controller%20Object%20Reference))
  ;[0 0.1812 1 0.2] Other Devices alter PlayerDefined stats by sending StatEvents through the MinigameStatsComponent (e.g. FortStatPowerupExecutionCalculation.cpp)
  ; For PlayerDefinedStats, we don't directly add them and instead these delegates are triggered in order to allow this device to sanitize the stat value before it's applied (e.g. Clamping it within limits, ignoring it if it's not valid)
  (CallFunction|ApplyValueToStatGroup InstigatingController Amount true))

(event Custom|On%20Stat%20Level%20Changed
  (params (Amount :type Integer) (InstigatingController :type Controller%20Object%20Reference))
  ;[0 0.1812 1 0.2] Other Devices alter PlayerDefined stats by sending StatEvents through the MinigameStatsComponent (e.g. FortStatPowerupExecutionCalculation.cpp)
  ; For PlayerDefinedStats, we don't directly add them and instead these delegates are triggered in order to allow this device to sanitize the stat value before it's applied (e.g. Clamping it within limits, ignoring it if it's not valid)
  (StatUpdating|ApplyLeveltoStatGroup InstigatingController Amount true true -1 true))

(event Custom|DisableEvent
  (params (Instigator :type Controller%20Object%20Reference))
  ;[0.4835 0 1 0.2] Channel Receivers
  (Class|CreativeEnabledComponent|SetEnabled (Variables|Default|GetEnabledComponent) false))

(event Custom|LevelIncreaseEvent
  (params (Instigator :type Controller%20Object%20Reference))
  ;[0.4835 0 1 0.2] Channel Receivers
  (StatUpdating|ApplyLeveltoStatGroup Instigator 1 true true -1 true))

(event Custom|SaveEvent
  (params (Instigator :type Controller%20Object%20Reference))
  ;[0.4835 0 1 0.2] Channel Receivers
  (Utilities|IsValid Instigator
    (:Is%20Valid
      (bind _playerstate (Variables|/Script/Engine.Controller|GetPlayerState Instigator)
        (:then
          (Utilities|IsEnabled
            (:true
              (Persistence|SaveStat _playerstate))))
        (:else)))))

(event OnEnabledStateChanged(EnabledComponent)
  (params (Enabled :type Boolean))
  ;[0 0.787 1 0] If disabled, we cache all of our stats into local variables. (New Stats won't apply as there's an enabled check in the update on 'Apply Value to Stat Group').
  (bind _statdetailscomponent (UI|FindorCreatePlayspaceComponent))
  (Utilities|IsValid _statdetailscomponent
    (:Is%20Valid
      (Class|PlayspaceComponentPlayerStatDetails|SetEnabled _statdetailscomponent (Class|FortCreativeStatManagerBlueprintLibrary|GetStatFilter self (Variables|Default|GetValueStatTag)) Enabled (Variables|Default|GetValueStatTag))
      (if Enabled
        (Cache|RestoreFromCache)
        (else
          (Cache|CacheandClearStats))))))

(event Custom|LevelDecreaseEvent
  (params (Instigator :type Controller%20Object%20Reference))
  ;[0.4835 0 1 0.2] Channel Receivers
  (StatUpdating|ApplyLeveltoStatGroup Instigator -1 true true -1 true))

(event Custom|LoadEvent
  (params (Instigator :type Controller%20Object%20Reference))
  ;[0.4835 0 1 0.2] Channel Receivers
  (Utilities|IsValid Instigator
    (:Is%20Valid
      (bind _playerstate (Variables|/Script/Engine.Controller|GetPlayerState Instigator)
        (:then
          (Utilities|IsEnabled
            (:true
              (Persistence|LoadStatAll _playerstate) ; Loads stats for all players relevant to the requesting Player State\n(e.g. if Scope:Player, only player. If Scope:Team, then all PS's teammates)
            )))
        (:else)))))

(event Custom|LoadForAllEvent
  (params (Instigator :type Controller%20Object%20Reference))
  ;[0.4835 0 1 0.2] Channel Receivers
  (Utilities|IsEnabled
    (:true
      (Persistence|LoadStatAll 0) ; Loads stats for all players relevant to the requesting Player State\n(e.g. if Scope:Player, only player. If Scope:Team, then all PS's teammates)
    )))

(event Custom|On%20Player%20Team%20Changed
  (params (PlayerState :type Fort%20Player%20State%20Object%20Reference) (TeamId :type Byte) (OldTeamId :type Byte) (bTeamChangedWithoutRespawn :type Boolean))
  ;[0 0.5531 1] Player State Changed. If necessary, update their stat
  (Default|ValidatePlayerandCacheOrRestoreStats PlayerState))

(event Custom|ResetStatEvent
  (params (Instigator :type Controller%20Object%20Reference))
  ;[0.4835 0 1 0.2] Channel Receivers
  (Utilities|IsEnabled
    (:true
      (CallFunction|SetRelevantStatValue (Class|FortCreativeStatManagerBlueprintLibrary|GetStatFilter self (Variables|Default|GetValueStatTag)) 0 Instigator)
      (CallFunction|SetRelevantStatValue (Class|FortCreativeStatManagerBlueprintLibrary|GetStatFilter self (Variables|Default|GetLevelStatTag)) 1 Instigator) ; Stat Value 1 is the default state for a level
    )))

(event Custom|On%20Player%20Class%20Changed
  (params (PlayerState :type Fort%20Player%20State%20Object%20Reference) (ClassSlotIndex :type Byte))
  ;[0 0.5531 1] Player State Changed. If necessary, update their stat
  (Default|ValidatePlayerandCacheOrRestoreStats PlayerState))

(event Custom|ClearPlayerPersistenceDataEvent
  (params (Instigator :type Controller%20Object%20Reference))
  ;[0.4835 0 1 0.2] Channel Receivers
  (Utilities|IsEnabled
    (:true
      (Utilities|IsValid Instigator
        (:Is%20Valid
          (bind _playerstate (Variables|/Script/Engine.Controller|GetPlayerState Instigator)
            (:then
              (Persistence|ClearData (Variables|Default|GetFortCreativePersistenceComponentBase) (PlayerState|GetUniqueNetId _playerstate) (Variables|Persistence|GetSavePathName)))
            (:else)))))))

(event Custom|ClearPlayerPersistenceDataForAllEvent
  (params (Instigator :type Controller%20Object%20Reference))
  ;[0.4835 0 1 0.2] Channel Receivers
  (Utilities|IsEnabled
    (:true
      (for _array_element (Persistence|GetPersistencePlayersIds (Variables|Default|GetFortCreativePersistenceComponentBase))
        (Persistence|ClearData (Variables|Default|GetFortCreativePersistenceComponentBase) _array_element (Variables|Persistence|GetSavePathName))))))

(event EventEndPlay
  (params (EndPlayReason :type EEndPlayReason%20Enum :default "Destroyed"))
  (bind _staticon (Variables|UserOptions|UI|GetStatIcon))
  (bind (_smallicon _largeicon) (Utilities|Struct|BreakCreativeIconOption _staticon))
  (bind _statcolor (Variables|UserOptions|UI|GetStatColor))
  (bind _statname (Variables|UserOptions|GetStatName))
  ;[1 0 0.00807 0.201] EndPlay
  (|CallParentFunction EndPlayReason)
  (if (not (Building|IsPreviewBuildingActor))
    (CallFunction|BindMinigame false)
    (CallFunction|SetupMinigameStat false)
    (if (not (Editor|InEditor))
      (bind _statdetailscomponent (UI|FindorCreatePlayspaceComponent))
      (Utilities|IsValid _statdetailscomponent
        (:Is%20Valid
          (bind _returnvalue (GUID|GenerateGuidfromObject self true))
          (Default|DeregisterDevice _statdetailscomponent _returnvalue)))
      (else
        (bind _outactors (Actor|GetAllActorswithInterface "/CRD_StatManager/BPi_AffectedByStat.BPi_AffectedByStat_C"))
        (for _array_element _outactors
          (StatDetails|UpdateStat(Message) _array_element (Variables|Default|GetValueStatTag) true _largeicon _statcolor _statname)
          (StatDetails|UpdateStat(Message) _array_element (Variables|Default|GetLevelStatTag) true _largeicon _statcolor _statname))))))

(event Custom|OnDestroyed_Event
  (params (DestroyedActor :type Actor%20Object%20Reference))
  (bind _staticon (Variables|UserOptions|UI|GetStatIcon))
  (bind (_smallicon _largeicon) (Utilities|Struct|BreakCreativeIconOption _staticon))
  (bind _statcolor (Variables|UserOptions|UI|GetStatColor))
  (bind _statname (Variables|UserOptions|GetStatName))
  ;[1 0 0.00807 0.201] EndPlay
  (if (not (Building|IsPreviewBuildingActor))
    (CallFunction|BindMinigame false)
    (CallFunction|SetupMinigameStat false)
    (if (not (Editor|InEditor))
      (bind _statdetailscomponent (UI|FindorCreatePlayspaceComponent))
      (Utilities|IsValid _statdetailscomponent
        (:Is%20Valid
          (bind _returnvalue (GUID|GenerateGuidfromObject self true))
          (Default|DeregisterDevice _statdetailscomponent _returnvalue)))
      (else
        (bind _outactors (Actor|GetAllActorswithInterface "/CRD_StatManager/BPi_AffectedByStat.BPi_AffectedByStat_C"))
        (for _array_element _outactors
          (StatDetails|UpdateStat(Message) _array_element (Variables|Default|GetValueStatTag) true _largeicon _statcolor _statname)
          (StatDetails|UpdateStat(Message) _array_element (Variables|Default|GetLevelStatTag) true _largeicon _statcolor _statname))))))

(event EventPostFinishDuplication
  (params (SourceObject :type Object%20Reference) (bReplacesSource :type Boolean))
  ;[1 0 0.759 0.5] Duplication Logic - Ensuring our stats maintain their naming correctly when cut / copied.
  (if (not bReplacesSource)
    (CallFunction|SetupMinigameStat false) ; Throwaway our generated stat if any,\nWe'll be stealing the one from the original device.
    (Variables|Default|SetValueStatTag)
    (Variables|Default|SetLevelStatTag)
    (Variables|Cache|SetOldStatName "")
    (CallFunction|InitializeDevice)
    (CallFunction|Resave)
    (else
      (Replacement|UpdatefromOtherStatCreator SourceObject))))

(event Custom|On%20Finished%20Adding%20Minigame%20Player
  (params (UniqueNetId :type Unique%20Net%20Id%20Structure) (bIsLocalPlayer :type Boolean) (FortPlayerState :type Fort%20Player%20State%20Object%20Reference))
  ;[0.04337 1 0.127 0.187] If we're joining in progress, load our stat, but only if the appropriate bucket doesn't already have a stat (Don't want to stomp on an in-progress stat).
  (if (and (Class|CreativeEnabledComponent|IsEnabled (Variables|Default|GetEnabledComponent)) (Variables|UserOptions|Persistence|GetUsePersistence))
    (if (not (Persistence|IsInventoryandPersistenceReadyforPlayer (Variables|Default|GetFortCreativePersistenceComponentBase) UniqueNetId))
      (Utilities|FlowControl|DelayUntilNextTick))
    (CallFunction|AddStatonJoin FortPlayerState)))

(event Custom|On%20Minigame%20Player%20Spawned
  (params (SpawningPlayer :type Controller%20Object%20Reference))
  (bind _playerstate (Variables|/Script/Engine.Controller|GetPlayerState SpawningPlayer))
  ;[0.04337 1 0.127 0.187] If we're joining in progress, load our stat, but only if the appropriate bucket doesn't already have a stat (Don't want to stomp on an in-progress stat).
  (if (and (Class|CreativeEnabledComponent|IsEnabled (Variables|Default|GetEnabledComponent)) (Variables|UserOptions|Persistence|GetUsePersistence))
    (if (not (Persistence|IsInventoryandPersistenceReadyforPlayer (Variables|Default|GetFortCreativePersistenceComponentBase) (PlayerState|GetUniqueNetId _playerstate)))
      (Utilities|FlowControl|DelayUntilNextTick))
    (CallFunction|AddStatonJoin _playerstate)))

(event Custom|Update%20Device%20Icon
  (bind _creativeeditonlymesh (Variables|Components|GetCreativeEditOnlyMesh))
  (bind _iconmid (Variables|Default|GetIconMID))
  (bind _statcolor (Variables|UserOptions|UI|GetStatColor))
  (bind (_staticon_smallicon _staticon_largeicon) (Utilities|Struct|BreakCreativeIconOption (Variables|GetStatIcon_SmallIcon)))
  (bind (_staticon_smallicon_2 _staticon_largeicon_2) (Utilities|Struct|BreakCreativeIconOption (Variables|GetStatIcon_SmallIcon)))
  ;[1 0.6099 0 0.2] Updating visuals for the device based on the stat and color.
  (Maerial|SetColor (Variables|Default|GetCreativeEditOnlyBase) _statcolor _statcolor)
  (Material|SetColor (Variables|Default|GetCreativeEditOnlyBeam) _statcolor)
  (Variables|/Script/Engine.StaticMeshComponent|GetStaticMesh _creativeeditonlymesh
    (:then
      (Networking|CanExecuteCosmeticEvents
        (:True
          (Material|ReattachOrCreateMID _creativeeditonlymesh _iconmid)
          (if (Utilities|IsValidSoftObjectReference _staticon_largeicon_2)
            (Rendering|Material|SetVectorParameterValue _iconmid "ColorBackground" (Variables|UserOptions|UI|GetStatColor) :!class /Script/Engine.MaterialInstanceDynamic)
            (bind _loaded_object (Creative|Loading|AsyncLoadEditorSafe _staticon_largeicon)
              (:Loaded
                (bind _astexture (Utilities|Casting|CastToTexture _loaded_object))
                (Rendering|Material|SetTextureParameterValue _iconmid "DeviceIcon" _astexture)))))))
    (:else)))

(event OnMeshandMaterialsLoaded(CreativeEditOnlyBeam)
  ;[1 0.6099 0 0.2] Updating the color of the icon once the material loads.
  (Material|SetColor (Variables|Default|GetCreativeEditOnlyBeam) (Variables|UserOptions|UI|GetStatColor)))

(event OnMeshandMaterialsLoaded(CreativeEditOnlyBase)
  ;[1 0.6099 0 0.2] Updating the color of the icon once the material loads.
  (Maerial|SetColor (Variables|Default|GetCreativeEditOnlyBase) (Variables|UserOptions|UI|GetStatColor) (Variables|UserOptions|UI|GetStatColor)))

(event OnMeshandMaterialsLoaded(CreativeEditOnlyMesh)
  ;[1 0.6099 0 0.2] Updating the color of the icon once the material loads.
  (CallFunction|UpdateDeviceIcon))

(fn AddStatOnJoin
  (params (PlayerInstigatorState :type Player%20State%20Object%20Reference))
  (bind _returnvalue (PlayerState|GetUniqueNetId PlayerInstigatorState))
  (if (Math|Boolean|ANDBoolean (Variables|UserOptions|Persistence|GetbAutoLoad) (Variables|UserOptions|Persistence|GetUsePersistence) (not (Utilities|Array|ContainsItem (Variables|Default|GetActivePlayers) _returnvalue)))
    (Persistence|LoadStat PlayerInstigatorState true)
    (Utilities|Array|Add (Variables|Default|GetActivePlayers) _returnvalue)))

(fn ApplyValuetoStatGroup
  (params (PlayerInstigator :type Controller%20Object%20Reference) (Value :type Integer) (Add%20Value :type Boolean) (ForceTeam :type Integer :default -1) (SkipValidityCheck :type Boolean))
  (Utilities|IsEnabled
    (:true
      (Utilities|ValidforStat PlayerInstigator ForceTeam
        (:True
          _)
        (:False
          (if (Variables|GetSkipValidityCheck)
            _)))
      (bind _output_get (LocalVariables|SetCurrentValue (Default|GetRelevantStatValue (Class|FortCreativeStatManagerBlueprintLibrary|GetStatFilter self (Variables|Default|GetValueStatTag)) PlayerInstigator ForceTeam)))
      (bind _output_get_2 (LocalVariables|SetNewValue (+ (Math|Integer|SelectInt _output_get 0 Add%20Value) (Variables|GetValue))))
      (if (!= _output_get_2 (LocalVariables|GetCurrentValue))
        (bind (_newvalue _newlevel) (CallFunction|CalculateNewValueandLevel (Default|GetRelevantStatValue (Class|FortCreativeStatManagerBlueprintLibrary|GetStatFilter self (Variables|Default|GetLevelStatTag)) PlayerInstigator ForceTeam) (LocalVariables|GetNewValue)))
        (Utilities|FlowControl|Sequence
          (:then_0
            (LocalVariables|SetNewValue _newvalue)
            (if (!= (LocalVariables|GetCurrentValue) (LocalVariables|GetNewValue))
              (CallFunction|SetRelevantStatValue (Class|FortCreativeStatManagerBlueprintLibrary|GetStatFilter self (Variables|Default|GetValueStatTag)) (LocalVariables|GetNewValue) (Variables|GetPlayerInstigator) (Variables|GetForceTeam)) ; We're above our level cap but have value. Clamp it.
              (CallFunction|BroadcastValueChange (Variables|GetValue) (Variables|GetPlayerInstigator) (Variables|GetForceTeam))
              ;[0.1297 1 0.8399 0.366] If we don't have levels but do have a maximum value, broadcast OnReachedMaximum when we hit our maximum limit.
              (Utilities|HasLevels
                (:No
                  (if (and (Variables|UserOptions|GetMaxValue_Override) (== (LocalVariables|GetNewValue) (Variables|UserOptions|GetMaxValue)))
                    (GameplayEvent|BroadcastEvent (Variables|UserOptions|Events|GetOn%20Reached%20Maximum) 0 (Utilities|Struct|MakeGameplayEventStatManagerLevelMaximum (Variables|GetPlayerInstigator))))))))
          (:then_1
            (StatUpdating|ApplyLeveltoStatGroup (Variables|GetPlayerInstigator) _newlevel false false (Variables|GetForceTeam) true (Variables|GetSkipValidityCheck))))))))

(fn ApplyLevelToStatGroup
  (params (Controller :type Controller%20Object%20Reference) (Level :type Integer) (AddLevel :type Boolean :default true) (Reset%20Value :type Boolean) (ForceTeam :type Integer :default -1) (Broadcast :type Boolean) (SkipValidCheck :type Boolean))
  (bind _currentlevel (LocalVariables|GetCurrentLevel))
  (bind _newlevel (LocalVariables|GetNewLevel))
  (bind _returnvalue (- _newlevel _currentlevel))
  (bind _controller (Variables|GetController))
  (Utilities|HasLevels
    (:Yes
      (Utilities|IsEnabled
        (:true
          (if (not SkipValidCheck)
            (Utilities|ValidforStat Controller ForceTeam
              (:True
                _)))
          (LocalVariables|SetCurrentLevel (Default|GetRelevantStatValue (Class|FortCreativeStatManagerBlueprintLibrary|GetStatFilter self (Variables|Default|GetLevelStatTag)) Controller ForceTeam))
          (bind _output_get (LocalVariables|SetNewLevel (Math|Integer|Max(Integer) (Math|Integer|Clamp(Integer) (+ (Variables|GetLevel) (Math|Integer|SelectInt (LocalVariables|GetCurrentLevel) 0 AddLevel)) 0 (LevelScaling|GetLevelNumber)) 1)))
          (if (!= (- _output_get (LocalVariables|GetCurrentLevel)) 0)
            (CallFunction|SetRelevantStatValue (Class|FortCreativeStatManagerBlueprintLibrary|GetStatFilter self (Variables|Default|GetLevelStatTag)) (Math|Integer|Max(Integer) (LocalVariables|GetNewLevel) 1) _controller (Variables|GetForceTeam))
            (if (Variables|GetBroadcast)
              (for _ (range 1 (+ (Math|Integer|Absolute(Integer) _returnvalue) 1))
                (CallFunction|BroadcastLevelChange _returnvalue _controller (Variables|GetForceTeam)))
              (else
                _))
            (if (Variables|GetReset%20Value)
              (CallFunction|SetRelevantStatValue (Class|FortCreativeStatManagerBlueprintLibrary|GetStatFilter self (Variables|Default|GetValueStatTag)) 0 (Variables|GetController) (Variables|GetForceTeam)))))))))

(fn BindMinigame
  (params (Bind :type Boolean))
  (bind _outputdelegate (EventDispatchers|CreateEvent :!event On%20Player%20Team%20Changed))
  (bind _outputdelegate_2 (EventDispatchers|CreateEvent :!event On%20Player%20Class%20Changed))
  (bind _outputdelegate_3 (EventDispatchers|CreateEvent :!event On%20Finished%20Adding%20Minigame%20Player))
  (bind _outputdelegate_4 (EventDispatchers|CreateEvent :!event On%20Finished%20Removing%20Minigame%20Player))
  (bind _outputdelegate_5 (EventDispatchers|CreateEvent :!event On%20Minigame%20Player%20Spawned))
  (bind _currentminigame (Variables|/Script/FortniteGame.FortMinigameLogicComponent|GetCurrentMinigame (Variables|Default|GetFortMinigameProgress))
    (:then
      (if Bind
        (Minigame|BindEventtoPlayerChangedTeamDelegate _currentminigame _outputdelegate)
        (Minigame|BindEventtoOnPlayerClassApplied _currentminigame _outputdelegate_2)
        (Minigame|BindEventtoOnFinishedAddingMinigamePlayerDelegate _currentminigame _outputdelegate_3)
        (Minigame|BindEventtoOnMinigamePlayerSpawned _currentminigame _outputdelegate_5)
        (Minigame|BindEventtoPlayerRemovedDelegate _currentminigame _outputdelegate_4)
        (else
          (Minigame|UnbindEventfromPlayerChangedTeamDelegate _currentminigame _outputdelegate)
          (Minigame|UnbindEventfromOnPlayerClassApplied _currentminigame _outputdelegate_2)
          (Minigame|UnbindEventfromOnFinishedAddingMinigamePlayerDelegate _currentminigame _outputdelegate_3)
          (Minigame|UnbindEventfromOnMinigamePlayerSpawned _currentminigame _outputdelegate_5)
          (Minigame|UnbindEventfromPlayerRemovedDelegate _currentminigame _outputdelegate_4))))
    (:else)))

(fn BindStatEvents
  (params (Bind :type Boolean))
  (bind _levelstattag (Variables|Default|GetLevelStatTag))
  (bind _self self)
  (bind _returnvalue (Class|FortCreativeStatManagerBlueprintLibrary|GetStatFilter _self _levelstattag))
  (bind _valuestattag (Variables|Default|GetValueStatTag))
  (bind _self_2 self)
  (bind _returnvalue_2 (Class|FortCreativeStatManagerBlueprintLibrary|GetStatFilter _self_2 _valuestattag))
  (bind _outputdelegate (EventDispatchers|CreateEvent :!event On%20Stat%20Level%20Changed))
  (bind _outputdelegate_2 (EventDispatchers|CreateEvent :!event On%20Stat%20Value%20Changed))
  (if Bind
    (bind _asfort_minigame_player_defined_stat_filter (Utilities|Casting|CastToFortMinigamePlayerDefinedStatFilter _returnvalue_2))
    (EventDispatchers|BindEventtoOnAddtoPlayerDefinedStatDelegate _asfort_minigame_player_defined_stat_filter _outputdelegate_2)
    (bind _asfort_minigame_player_defined_stat_filter_2 (Utilities|Casting|CastToFortMinigamePlayerDefinedStatFilter _returnvalue))
    (EventDispatchers|BindEventtoOnAddtoPlayerDefinedStatDelegate _asfort_minigame_player_defined_stat_filter_2 _outputdelegate)
    (else
      (bind _asfort_minigame_player_defined_stat_filter_3 (Utilities|Casting|CastToFortMinigamePlayerDefinedStatFilter _returnvalue_2))
      (EventDispatchers|UnbindEventfromOnAddtoPlayerDefinedStatDelegate _asfort_minigame_player_defined_stat_filter_3 _outputdelegate_2)
      (bind _asfort_minigame_player_defined_stat_filter_4 (Utilities|Casting|CastToFortMinigamePlayerDefinedStatFilter _returnvalue))
      (EventDispatchers|UnbindEventfromOnAddtoPlayerDefinedStatDelegate _asfort_minigame_player_defined_stat_filter_4 _outputdelegate))))

(fn BroadcastLevelChange
  (params (LevelChangeSign :type Integer) (Controller :type Controller%20Object%20Reference) (ForceTeam :type Integer))
  (if (!= LevelChangeSign 0)
    (bind _players (CallFunction|GetRelevantPlayers Controller ForceTeam))
    (for _array_element _players
      (Utilities|IsValid _array_element
        (:Is%20Valid
          (LocalVariables|SetPlayerController (PlayerState|GetPlayerController _array_element)))
        (:Is%20Not%20Valid
          _))
      (bind _output_get (LocalVariables|SetNewLevel (Default|GetRelevantStatValue (Class|FortCreativeStatManagerBlueprintLibrary|GetStatFilter self (Variables|Default|GetLevelStatTag)) (LocalVariables|GetPlayerController) ForceTeam)))
      (GameplayEvent|BroadcastEvent (Variables|UserOptions|Events|GetOn%20Level%20Changed) 0 (Utilities|Struct|MakeGameplayEventStatManagerLevelChanged (LocalVariables|GetPlayerController) _output_get))
      (if (>= LevelChangeSign 0)
        ;[1 1 1] On Level Changed Transmission
        (if (>= (LocalVariables|GetNewLevel) (LevelScaling|GetLevelNumber))
          (GameplayEvent|BroadcastEvent (Variables|UserOptions|Events|GetOn%20Reached%20Maximum) 0 (Utilities|Struct|MakeGameplayEventStatManagerLevelMaximum (LocalVariables|GetPlayerController) (LocalVariables|GetNewLevel)))
          (else
            _))
        (GameplayEvent|BroadcastEvent (Variables|UserOptions|Events|GetOn%20Level%20Up) 0 (Utilities|Struct|MakeGameplayEventStatManagerLevelUp (LocalVariables|GetPlayerController) (LocalVariables|GetNewLevel)))
        (else
          (GameplayEvent|BroadcastEvent (Variables|UserOptions|Events|GetOn%20Level%20Down) 0 (Utilities|Struct|MakeGameplayEventStatManagerLevelDown (LocalVariables|GetPlayerController) (LocalVariables|GetNewLevel))))))))

(fn BroadcastValueChange
  (params (ValueSign :type Integer) (Controller :type Controller%20Object%20Reference) (ForceTeam :type Integer))
  (if (!= ValueSign 0)
    (bind _players (CallFunction|GetRelevantPlayers Controller ForceTeam))
    (for _array_element _players
      (Utilities|IsValid _array_element
        (:Is%20Valid
          (LocalVariables|SetPlayerController (PlayerState|GetPlayerController _array_element)))
        (:Is%20Not%20Valid
          _))
      (GameplayEvent|BroadcastEvent (Variables|UserOptions|Events|GetOn%20Value%20Changed) 0 (Utilities|Struct|MakeGameplayEventStatManagerValueUp (LocalVariables|GetPlayerController) (Default|GetRelevantStatValue (Class|FortCreativeStatManagerBlueprintLibrary|GetStatFilter self (Variables|Default|GetValueStatTag)) Controller ForceTeam))))))

(fn CacheAndClearStats
  (bind _levelstattag (Variables|Default|GetLevelStatTag))
  (bind _self self)
  (bind _returnvalue (Class|FortCreativeStatManagerBlueprintLibrary|GetStatFilter _self _levelstattag))
  (bind _valuestattag (Variables|Default|GetValueStatTag))
  (bind _self_2 self)
  (bind _returnvalue_2 (Class|FortCreativeStatManagerBlueprintLibrary|GetStatFilter _self_2 _valuestattag))
  (bind _output_get (LocalVariables|SetStatsComponent (Minigame|Stats|GetMinigameStatsComponent self)))
  (Utilities|IsValid _output_get
    (:Is%20Valid
      (bind _currentminigame (Variables|/Script/FortniteGame.FortMinigameLogicComponent|GetCurrentMinigame (Variables|Default|GetFortMinigameProgress))
        (:then
          (Utilities|IsValid (Minigame|GetMinigameTeamComponent _currentminigame)
            (:Is%20Valid
              (switch Utilities|FlowControl|Switch|SwitchonE_StatManager_Scope (Variables|UserOptions|GetScope)
                (:NewEnumerator0
                  (bind _outplayers (Minigame|GetParticipatingPlayers_NotPure _currentminigame))
                  (for _array_element _outplayers
                    (Utilities|ValidforStat _array_element
                      (:True
                        (Cache|CachePlayerStat (PlayerState|GetUniqueNetId _array_element) (Variables|StatCache|GetCache_PlayerValue) (Class|FortCreativeStatManagerBlueprintLibrary|GetStatFilter self (Variables|Default|GetValueStatTag)))
                        (Utilities|HasLevels
                          (:Yes
                            (Cache|CachePlayerStat (PlayerState|GetUniqueNetId _array_element) (Variables|StatCache|GetCache_PlayerLevel) (Class|FortCreativeStatManagerBlueprintLibrary|GetStatFilter self (Variables|Default|GetLevelStatTag)))))))))
                (:NewEnumerator1
                  (for _array_element_2 (Variables|/Script/FortniteGame.FortMinigameTeamComponent|GetTeamArray (Minigame|GetMinigameTeamComponent _currentminigame))
                    (if (.bHasBucketAvailable (Utilities|Struct|BreakMinigameTeam _array_element_2))
                      (Utilities|ValidforStat 0 (Teams|TeamtoIndex (Utilities|Struct|BreakMinigameTeam _array_element_2))
                        (:True
                          (Cache|CacheTeamStat (Variables|StatCache|GetCache_TeamValue) (Teams|TeamtoIndex (Utilities|Struct|BreakMinigameTeam _array_element_2)) (Class|FortCreativeStatManagerBlueprintLibrary|GetStatFilter self (Variables|Default|GetValueStatTag)))
                          (Utilities|HasLevels
                            (:Yes
                              (Cache|CacheTeamStat (Variables|StatCache|GetCache_TeamLevel) (Teams|TeamtoIndex (Utilities|Struct|BreakMinigameTeam _array_element_2)) (Class|FortCreativeStatManagerBlueprintLibrary|GetStatFilter self (Variables|Default|GetLevelStatTag))))))))))
                (:NewEnumerator2
                  (Variables|StatCache|SetCache_MatchValue (Minigame|Stats|GetStatforGroup (LocalVariables|GetStatsComponent) _returnvalue_2 false))
                  (Minigame|Stats|SetStatforGroup (LocalVariables|GetStatsComponent) _returnvalue_2 0 false)
                  (Utilities|HasLevels
                    (:Yes
                      (Variables|StatCache|SetCache_MatchLevel (Minigame|Stats|GetStatforGroup (LocalVariables|GetStatsComponent) _returnvalue false))
                      (Minigame|Stats|SetStatforGroup (LocalVariables|GetStatsComponent) _returnvalue 0 false))))))))
        (:else)))))

(fn CachePlayerStat
  (params (PlayerNetID :type Unique%20Net%20Id%20Structure) (Cache :type Map%20of%20Unique%20Net%20Id%20Structures%20to%20Integers) (StatFilter :type Fort%20Minigame%20Stat%20Filter%20Object%20Reference))
  (bind _statscomponent (LocalVariables|GetStatsComponent))
  (bind _returnvalue (Minigame|Stats|GetStatforPlayerId _statscomponent PlayerNetID StatFilter false))
  (bind _output_get (LocalVariables|SetStatsComponent (Minigame|Stats|GetMinigameStatsComponent self)))
  (Utilities|IsValid _output_get
    (:Is%20Valid
      (CallFunction|LogStatShipping PlayerNetID StatFilter _returnvalue false "Cache Player Stat")
      (Utilities|Map|Add Cache PlayerNetID _returnvalue)
      (CallFunction|LogStatShipping PlayerNetID StatFilter 0 true "Cache Player Stat")
      (Minigame|Stats|SetStatforPlayerId (LocalVariables|GetStatsComponent) PlayerNetID StatFilter 0 false) ; Setting current stat to 0, since device is disabled.
    )))

(fn CacheTeamStat
  (params (TeamMap :type Map%20of%20Integers%20to%20Integers) (BucketIndex :type Integer) (StatFitler :type Fort%20Minigame%20Stat%20Filter%20Object%20Reference))
  (bind _output_get (LocalVariables|SetStatsComponent (Minigame|Stats|GetMinigameStatsComponent self)))
  (Utilities|IsValid _output_get
    (:Is%20Valid
      (Utilities|Map|Add TeamMap BucketIndex (Minigame|Stats|GetStatforPlayerBucket (LocalVariables|GetStatsComponent) BucketIndex StatFitler false false))
      (Minigame|Stats|SetStatforPlayerBucket (LocalVariables|GetStatsComponent) BucketIndex StatFitler 0 false))))

(fn CalculateExpTables
  (bind _array (Utilities|Array|MakeArray 0 0))
  (Variables|Cache|SetExpNeededPerLevel _array)
  (Variables|Cache|SetCumulativeExpNeededPerLevel _array)
  (if (Variables|UserOptions|GetMaxValue_Override)
    (Utilities|HasLevels
      (:Yes
        (for _index (range 2 (+ (LevelScaling|GetLevelNumber) 1))
          (if (Math|Boolean|ORBoolean (== _index 2) (Utilities|IsValid (Variables|StatCurve|GetPointsPerLevelCurve) :!class /Script/Engine.KismetSystemLibrary) (!= (Variables|UserOptions|GetPerLevelPointsMultiplier) 1.0))
            (Utilities|Array|Add (Variables|Cache|GetExpNeededPerLevel) (LevelScaling|GetDeltaValueforLevel _index))
            (else
              _))
          (bind _output_get (LocalVariables|SetCumulatedTotal (+ (LevelScaling|GetDeltaValueforLevel _index) (LocalVariables|GetCumulatedTotal))))
          (Utilities|Array|Add (Variables|Cache|GetCumulativeExpNeededPerLevel) _output_get)))
      (:No
        ;[1 0.6099 0 0.2] If we don't have levels, store the Max Value into the Xp Per Level value
        ; (the UI needs to know what the 'next level' is, even if it never ticks over.
        (Utilities|Array|Add (Variables|Cache|GetExpNeededPerLevel) (Variables|UserOptions|GetMaxValue))
        (Variables|Cache|SetCumulativeExpNeededPerLevel (Variables|Cache|GetExpNeededPerLevel))))))

(fn CalculateNewValueAndLevel
  (params (CurrentLevel :type Integer) (SetValue :type Integer))
  (bind _trackinglevel (LocalVariables|GetTrackingLevel))
  (LocalVariables|SetTrackingLevel CurrentLevel)
  (if (== SetValue 0)
    ;[0.7542 0.6486 1 0.2] If our set value is 0, we don't have to bother with any validation, just return it.
    (return 0 (LocalVariables|GetTrackingLevel))
    (else
      (LocalVariables|SetValueToSet SetValue)
      (Utilities|HasLevels
        (:Yes
          (if (< (LocalVariables|GetValueToSet) 0)
            (if (Variables|UserOptions|GetbCanLoseLevelFromPointLoss)
              ;[0.7542 0.6486 1 0.2] While our Value is still negative (e.g. we want to progress down to a lower level)
              (while (and (< (LocalVariables|GetValueToSet) 0) (> (LocalVariables|GetTrackingLevel) 0))
                ;[0.7542 0.6486 1 0.2] Add the amount required to decrement a level, then decrement the level.
                (LocalVariables|SetValueToSet (+ (LevelScaling|GetDeltaValueforLevel (LocalVariables|GetTrackingLevel) true) (LocalVariables|GetValueToSet)))
                (Math|Integer|DecrementInt (LocalVariables|GetTrackingLevel)))
              (if (> (LocalVariables|GetTrackingLevel) 0)
                (return (Math|Integer|Max(Integer) (LocalVariables|GetValueToSet)) (LocalVariables|GetTrackingLevel))
                (else
                  (return 0 1)))
              (else
                (return 0 (LocalVariables|GetTrackingLevel))))
            (else
              (LocalVariables|SetDeltaLevelValue (LevelScaling|GetDeltaValueforLevel (+ (LocalVariables|GetTrackingLevel) 1) true))
              (while (and (> (LocalVariables|GetValueToSet) (- (LocalVariables|GetDeltaLevelValue) 1)) (!= (LocalVariables|GetTrackingLevel) (LevelScaling|GetLevelNumber)))
                (LocalVariables|SetValueToSet (- (LocalVariables|GetValueToSet) (LocalVariables|GetDeltaLevelValue)))
                (Math|Integer|IncrementInt _trackinglevel
                  (:%20%20
                    (LocalVariables|SetDeltaLevelValue (LevelScaling|GetDeltaValueforLevel (+ _trackinglevel 1) true)))))
              (return (Math|Integer|SelectInt 0 (LocalVariables|GetValueToSet) (== (LevelScaling|GetLevelNumber) (LocalVariables|GetTrackingLevel))) (LocalVariables|GetTrackingLevel)))))
        (:No
          ;[0.7542 0.6486 1 0.2] If we don't have levels, simply return the set value, clamped to our limits.
          (return (Math|Integer|SelectInt (Math|Integer|Max(Integer) SetValue) (Math|Integer|Clamp(Integer) SetValue 0 (Variables|UserOptions|GetMaxValue)) (not (Variables|UserOptions|GetMaxValue_Override))) 0))))))

(fn CanPlayerAutoSave
  (params (FortPlayerState :type Fort%20Player%20State%20Object%20Reference))
  (bind _valuestattag (Variables|Default|GetValueStatTag))
  (bind _self self)
  (bind _returnvalue (Class|FortCreativeStatManagerBlueprintLibrary|GetStatFilter _self _valuestattag))
  (bind _levelstattag (Variables|Default|GetLevelStatTag))
  (bind _self_2 self)
  (bind _returnvalue_2 (Class|FortCreativeStatManagerBlueprintLibrary|GetStatFilter _self_2 _levelstattag))
  (bind _returnvalue_3 (!= _returnvalue_2 0))
  (bind _returnvalue_4 (!= _returnvalue 0))
  (bind _returnvalue_5 (or _returnvalue_4 _returnvalue_3))
  (bind _returnvalue_6 (Development|GetConsoleVariableBoolValue "FortCreativeStatManager.EnableCacheAwareAutoSave"))
  (if _returnvalue_6
    (return (and (or (not (Utilities|Set|ContainsItem (Variables|Default|GetInvalidPlayers) FortPlayerState)) (Utilities|Map|Contains (Variables|StatCache|GetCache_PlayerValue) (PlayerState|GetUniqueNetId FortPlayerState))) _returnvalue_5))
    (else
      (return _returnvalue_5))))

(fn ClearCaches
  (Variables|StatCache|SetCache_MatchValue 0)
  (Variables|StatCache|SetCache_MatchLevel 0)
  (Utilities|Map|Clear (Variables|StatCache|GetCache_PlayerLevel))
  (Utilities|Map|Clear (Variables|StatCache|GetCache_TeamLevel))
  (Utilities|Map|Clear (Variables|StatCache|GetCache_PlayerValue))
  (Utilities|Map|Clear (Variables|StatCache|GetCache_TeamValue))
  (Utilities|Set|Clear (Variables|Default|GetInvalidPlayers)))

(fn FindOrCreatePlayspaceComponent
  (Networking|SwitchHasAuthority
    (:Authority
      ;[1 1 1] Generate Component if Necessary
      (Variables|Default|GetStatDetailsComponent
        (:then
          _)
        (:else
          (bind _returnvalue (Class|PlayspaceLibrary|GetPlayspaceforActor self self))
          (Utilities|IsValid _returnvalue
            (:Is%20Valid
              (bind _output_get (Variables|Default|SetStatDetailsComponent (Actor|GetComponentbyClass _returnvalue "/CRD_StatManager/UI/PlayspaceComponent_PlayerStatDetails.PlayspaceComponent_PlayerStatDetails_C")))
              (Utilities|IsValid _output_get
                (:Is%20Valid
                  _)
                (:Is%20Not%20Valid
                  (bind _returnvalue_2 (AddComponent|NODE_AddPlayspaceComponent_PlayerStatDetails-0 :self _returnvalue))
                  (Variables|Default|SetStatDetailsComponent _returnvalue_2)))
              (return (Variables|Default|GetStatDetailsComponent)))
            (:Is%20Not%20Valid
              (return 0))))))))

(fn GetAllMembersforGroup
  (params (Group%20Type :type E_StatManager_Scope%20Enum :default "NewEnumerator0") (Instigator :type Controller%20Object%20Reference) (ForceTeam :type Integer :default -1))
  (bind _fortminigameprogress (Variables|Default|GetFortMinigameProgress))
  (bind _currentminigame (Variables|/Script/FortniteGame.FortMinigameLogicComponent|GetCurrentMinigame _fortminigameprogress))
  (bind _returnvalue (Minigame|GetMinigameTeamComponent _currentminigame))
  (bind _forceteam (Variables|GetForceTeam))
  (bind _team (Teams|IndextoTeam _forceteam))
  ;[0 0.787 1 0.2] Used for Verse calls. If we call 'Set for Team' and we're player scoped, we'll want to set for all members of the team. Same for 'Everyone' if player or team scoped.
  (Utilities|FlowControl|Sequence
    (:then_0
      (Utilities|IsValid _currentminigame
        (:Is%20Valid
          (Utilities|IsValid _returnvalue
            (:Is%20Valid
              (switch Utilities|FlowControl|Switch|SwitchonE_StatManager_Scope Group%20Type
                (:NewEnumerator1
                  (bind _output_get (LocalVariables|SetPlayerTeam _team))
                  (if (>= _output_get 3)
                    (switch Utilities|FlowControl|Switch|SwitchonE_StatManager_Scope (Variables|UserOptions|GetScope)
                      (:NewEnumerator0
                        (return (Class|FortMinigameTeamComponent|GetTeamMembers _returnvalue _team) (Utilities|Array|MakeArray) false))
                      (:NewEnumerator1
                        (return (Utilities|Array|MakeArray) (Utilities|Array|MakeArray (LocalVariables|GetPlayerTeam)) false)))))
                (:NewEnumerator2
                  (switch Utilities|FlowControl|Switch|SwitchonE_StatManager_Scope (Variables|UserOptions|GetScope)
                    (:NewEnumerator0
                      (bind _outplayers (Minigame|GetParticipatingPlayers_NotPure (Variables|/Script/FortniteGame.FortMinigameLogicComponent|GetCurrentMinigame (Variables|Default|GetFortMinigameProgress))))
                      (return _outplayers (Utilities|Array|MakeArray) false))
                    (:NewEnumerator1
                      (return (Utilities|Array|MakeArray) (CallFunction|GetMinigameTeams) false))
                    (:NewEnumerator2
                      (return (Utilities|Array|MakeArray) (Utilities|Array|MakeArray) true))))))
            (:Is%20Not%20Valid
              _)))
        (:Is%20Not%20Valid
          _))
      (return (Utilities|Array|MakeArray 0) (Utilities|Array|MakeArray) false))
    (:then_1
      (Utilities|IsValid Instigator
        (:Is%20Valid
          (bind _asfort_player_state (Utilities|Casting|CastToFortPlayerState (Variables|/Script/Engine.Controller|GetPlayerState Instigator))
            (:then
              (return (Utilities|Array|MakeArray _asfort_player_state) (Utilities|Array|MakeArray) false))
            (:CastFailed
              _)))
        (:Is%20Not%20Valid
          _))
      (return (Utilities|Array|MakeArray 0) (Utilities|Array|MakeArray) false))))

(fn GetEnabledStateasIndex
  (return (DeviceConversion|EnabledPhaseEnumtoLegacyInt (Variables|UserOptions|GetEnabled%20Phase))))

(fn GetAllLevel
  (return (Default|GetRelevantStatValue (Class|FortCreativeStatManagerBlueprintLibrary|GetStatFilter self (Variables|Default|GetLevelStatTag)))))

(fn GetAllValue
  ;[0 0.787 1 0.2] If the Scope of the stat is 'Match', return the match stat. Else return 0.
  (if (Utilities|Enum|Equal(Enum) (Variables|UserOptions|GetScope) "NewEnumerator2")
    (return (Default|GetRelevantStatValue (Class|FortCreativeStatManagerBlueprintLibrary|GetStatFilter self (Variables|Default|GetValueStatTag))))
    (else
      (return 0))))

(fn GetBestValuesToLoad
  (params (PlayerStates :type Array%20of%20Player%20State%20Object%20References))
  (bind _allvalues (LocalVariables|GetAllValues))
  (bind (_indexofminvalue _minvalue) (Math|Integer|MinOfIntArray _allvalues))
  (bind (_indexofmaxvalue _maxvalue) (Math|Integer|MaxOfIntArray _allvalues))
  (bind _persistencedata_statmanager (Variables|Persistence|GetPersistenceData_StatManager))
  (Utilities|FlowControl|Sequence
    (:then_0
      (bind _array_element (Utilities|Array|ForEachLoopwithBreak PlayerStates)
        (:LoopBody
          (Variables|/CRD_StatManager/SetupAssets/Persistence/PersistenceData_StatManager.PersistenceData_StatManager_C|SetbDataIsLoaded false _persistencedata_statmanager)
          (Persistence|LoadData (Variables|Default|GetFortCreativePersistenceComponentBase) (PlayerState|GetUniqueNetId _array_element) (Variables|Persistence|GetSavePathName) _persistencedata_statmanager)
          (if (Variables|/CRD_StatManager/SetupAssets/Persistence/PersistenceData_StatManager.PersistenceData_StatManager_C|GetbDataIsLoaded _persistencedata_statmanager)
            (Utilities|Array|Add (LocalVariables|GetAllValues) (+ (Variables|/CRD_StatManager/SetupAssets/Persistence/PersistenceData_StatManager.PersistenceData_StatManager_C|GetStatValue _persistencedata_statmanager) (LevelScaling|GetCumulatedValueforLevel (Variables|/CRD_StatManager/SetupAssets/Persistence/PersistenceData_StatManager.PersistenceData_StatManager_C|GetStatLevel _persistencedata_statmanager))))
            (if (Utilities|Enum|Equal(Enum) (Variables|UserOptions|Persistence|GetResolveConflicts) "NewEnumerator0")
              _)))
        (:Completed
          _))
      (if (Utilities|Array|IsNotEmpty (LocalVariables|GetAllValues))
        (switch Utilities|FlowControl|Switch|SwitchonE_StatManager_ResolveConflicts (Variables|UserOptions|Persistence|GetResolveConflicts)
          (:NewEnumerator0
            (LocalVariables|SetBestValue (Utilities|Array|Get(aref) (LocalVariables|GetAllValues))))
          (:NewEnumerator1
            (LocalVariables|SetBestValue _maxvalue))
          (:NewEnumerator2
            (LocalVariables|SetBestValue _minvalue))
          (:NewEnumerator3
            (LocalVariables|SetBestValue (Math|Float|Truncate (Math|Integer|AverageOfIntArray (LocalVariables|GetAllValues)))))
          (:NewEnumerator4
            (LocalVariables|SetBestValue (Math|Float|Round (Math|Integer|MedianOfIntArray (LocalVariables|GetAllValues))))))))
    (:then_1
      (bind (_outvalue _outlevel) (LevelScaling|GetLevelforValue (LocalVariables|GetBestValue)))
      (return _outvalue _outlevel))))

(fn GetBucketIndexForActor
  (params (Actor :type Actor%20Object%20Reference) (Force%20Team :type Integer))
  (bind _force_team (Variables|GetForce%20Team))
  (bind _returnvalue (>= _force_team 0))
  (bind _playerteamindex (LocalVariables|GetPlayerTeamIndex))
  (bind _returnvalue_2 (Math|Integer|SelectInt _force_team _playerteamindex _returnvalue))
  (Utilities|IsValid Actor
    (:Is%20Valid
      (LocalVariables|SetPlayerTeamIndex (Teams|TeamtoIndex (Team|GetActorTeam Actor))))
    (:Is%20Not%20Valid
      _))
  (Utilities|ValidforStat 0 _returnvalue_2
    (:True
      (return true _returnvalue_2))
    (:False
      (return false 0))))

(fn GetCollisionStaticMesh
  (return "0, 0, 0" 0))

(fn GetCumulatedValueForLevel
  (params (Level :type Integer))
  (bind _cumulativeexpneededperlevel (Variables|Cache|GetCumulativeExpNeededPerLevel))
  ;[1 1 1] Returns the experience required for a level. 'InRange' is false if you are asking for a level outside the valid level range (but it'll still return a valid result, clamped to 0 and max level)
  (if (and (>= Level 2) (Utilities|Array|IsNotEmpty _cumulativeexpneededperlevel))
    (return (Utilities|Array|Get(acopy) (Variables|Cache|GetCumulativeExpNeededPerLevel) (Math|Integer|Clamp(Integer) Level 0 (Utilities|Array|LastIndex _cumulativeexpneededperlevel))))
    (else
      (return 0))))

(fn GetDeltaValueForLevel
  (params (Level :type Integer) (UseCache :type Boolean))
  (if UseCache
    (return (Utilities|Array|Get(aref) (Variables|Cache|GetExpNeededPerLevel) (Math|Integer|Clamp(Integer) Level 0 (Utilities|Array|LastIndex (Variables|Cache|GetExpNeededPerLevel)))))
    (else
      (Utilities|FlowControl|Sequence
        (:then_0
          (bind _pointsperlevelcurve (Variables|StatCurve|GetPointsPerLevelCurve)
            (:then
              (LocalVariables|SetCurveMultiplier (Math|Curves|GetFloatValue _pointsperlevelcurve (Math|Conversions|ToFloat(Integer) Level))))
            (:else)))
        (:then_1
          (return (Math|Float|Truncate (Utilities|Operators|Multiply (+ (* (Math|Integer|Max(Integer) (- Level 2)) (- (Variables|UserOptions|GetPerLevelPointsMultiplier) 1.0)) 1.0) (Variables|UserOptions|GetMaxValue) (LocalVariables|GetCurveMultiplier)))))))))

(fn GetLevelForValue
  (params (CumulatedValue :type Integer))
  (Utilities|HasLevels
    (:Yes
      (for (_array_element _array_index) (Variables|Cache|GetCumulativeExpNeededPerLevel)
        (LocalVariables|SetLevel (- _array_index 1)) ; The array contains the amount needed to hit the NEXT level,\nso we're actually on the previous level.
        (if (< (- (Variables|GetCumulatedValue) _array_element) 0)
          (return (- (Variables|GetCumulatedValue) (LocalVariables|GetCurrentLevelValue)) (LocalVariables|GetLevel))
          (else
            (LocalVariables|SetCurrentLevelValue _array_element))))
      (return 0 (Variables|UserOptions|GetMaxLevel)))
    (:No
      (return CumulatedValue 0))))

(fn GetLevelNumber
  ;[0 0.787 1 0.2] Number of Levels is only set if MaxPoints is used. (Can't have levels if you don't have a value to define a level)
  (return (Math|Integer|SelectInt (Variables|UserOptions|GetMaxLevel) 1 (Variables|UserOptions|GetMaxValue_Override))))

(fn GetMeshComponents
  (return 0))

(fn GetMinigameTeams
  (bind _currentminigame (Variables|/Script/FortniteGame.FortMinigameLogicComponent|GetCurrentMinigame (Variables|Default|GetFortMinigameProgress))
    (:then
      (Utilities|IsValid (Minigame|GetMinigameTeamComponent _currentminigame)
        (:Is%20Valid
          (bind _output_get (LocalVariables|SetTeamList (Variables|/Script/FortniteGame.FortMinigameTeamComponent|GetTeamArray (Minigame|GetMinigameTeamComponent _currentminigame))))
          (for _array_element _output_get
            (bind (_array_element_teamindex _array_element_teamname _array_element_teamdescription _array_element_teamcolorindex _array_element_maxinitteamsize _array_element_initteamsizeweight _array_element_bhasbucketavailable _array_element_eliminatedcount _array_element_teamsize) (Utilities|Struct|BreakMinigameTeam _array_element))
            (if (and (> _array_element_teamsize 0) _array_element_bhasbucketavailable)
              (Utilities|Array|Add (LocalVariables|GetAllTeams) _array_element_teamindex)))
          (return (LocalVariables|GetAllTeams)))
        (:Is%20Not%20Valid
          _)))
    (:else
      _))
  (return 0))

(fn GetName
  (return (Utilities|String|ToString(Text) (Variables|UserOptions|GetStatName))))

(fn GetObjectToAutoSaveForPlayer
  (params (FortPlayerState :type Fort%20Player%20State%20Object%20Reference))
  (bind _returnvalue (PlayerState|GetUniqueNetId FortPlayerState))
  (bind _cache_playervalue (Variables|StatCache|GetCache_PlayerValue))
  (bind (_value _returnvalue_2) (Utilities|Map|Find _cache_playervalue _returnvalue))
  (bind _persistencedata_statmanager (Variables|Persistence|GetPersistenceData_StatManager))
  (bind _returnvalue_3 (Development|GetConsoleVariableBoolValue "FortCreativeStatManager.EnableCacheAwareAutoSave"))
  (if (and _returnvalue_3 (and (Utilities|Set|ContainsItem (Variables|Default|GetInvalidPlayers) FortPlayerState) _returnvalue_2))
    (Persistence|ResetPersistenceObject)
    (Variables|/CRD_StatManager/SetupAssets/Persistence/PersistenceData_StatManager.PersistenceData_StatManager_C|SetStatValue _value _persistencedata_statmanager)
    (Variables|/CRD_StatManager/SetupAssets/Persistence/PersistenceData_StatManager.PersistenceData_StatManager_C|SetStatLevel (Utilities|Map|Find (Variables|StatCache|GetCache_PlayerLevel) _returnvalue) _persistencedata_statmanager)
    (Variables|/CRD_StatManager/SetupAssets/Persistence/PersistenceData_StatManager.PersistenceData_StatManager_C|SetbDataIsLoaded true _persistencedata_statmanager)
    (return _persistencedata_statmanager)
    (else
      (Persistence|PopulatePersistenceDataObject (PlayerState|GetPlayerController FortPlayerState))
      (return _persistencedata_statmanager))))

(fn GetPersistenceRequiredState
  (return (Variables|UserOptions|Persistence|GetUsePersistence)))

(fn GetPlayerLevel
  (params (Controller :type Controller%20Object%20Reference))
  (return (Default|GetRelevantStatValue (Class|FortCreativeStatManagerBlueprintLibrary|GetStatFilter self (Variables|Default|GetLevelStatTag)) Controller)))

(fn GetPlayerValue
  (params (Controller :type Controller%20Object%20Reference))
  ;[0 0.787 1 0.2] Gets the stat value for a specific player.
  (return (Default|GetRelevantStatValue (Class|FortCreativeStatManagerBlueprintLibrary|GetStatFilter self (Variables|Default|GetValueStatTag)) Controller)))

(fn GetRelevantPlayers
  (params (Controller :type Controller%20Object%20Reference) (ForceTeam :type Integer))
  (bind _fortminigameprogress (Variables|Default|GetFortMinigameProgress))
  (bind _currentminigame (Variables|/Script/FortniteGame.FortMinigameLogicComponent|GetCurrentMinigame _fortminigameprogress))
  (bind _returnvalue (Minigame|GetMinigameTeamComponent _currentminigame))
  (Utilities|IsValid Controller
    (:Is%20Valid
      (Utilities|IsValid _currentminigame
        (:Is%20Valid
          (Utilities|IsValid _returnvalue
            (:Is%20Valid
              (if (Variables|UserOptions|GetBroadcastEventsForAllPlayers)
                (switch Utilities|FlowControl|Switch|SwitchonE_StatManager_Scope (Variables|UserOptions|GetScope)
                  (:NewEnumerator0
                    _)
                  (:NewEnumerator1
                    (return (Class|FortMinigameTeamComponent|GetTeamMembers _returnvalue (select (>= ForceTeam 0) (Math|Conversions|ToByte(Integer) ForceTeam) (Team|GetActorPvPTeamIndex Controller)))))
                  (:NewEnumerator2
                    (bind _outplayers (Minigame|GetParticipatingPlayers_NotPure (Variables|/Script/FortniteGame.FortMinigameLogicComponent|GetCurrentMinigame (Variables|Default|GetFortMinigameProgress))))
                    (return _outplayers)))
                (else
                  _))
              (bind _asfort_player_state (Utilities|Casting|CastToFortPlayerState (Variables|/Script/Engine.Controller|GetPlayerState Controller))
                (:then
                  (return (Utilities|Array|MakeArray _asfort_player_state)))
                (:CastFailed
                  (return (Utilities|Array|MakeArray)))))
            (:Is%20Not%20Valid
              _)))
        (:Is%20Not%20Valid
          _))
      (return (Utilities|Array|MakeArray 0)))
    (:Is%20Not%20Valid
      _)))

(fn GetRelevantStatValue
  (params (StatFilter :type Fort%20Minigame%20Stat%20Filter%20Object%20Reference) (PlayerInstigator :type Controller%20Object%20Reference) (ForceTeam :type Integer :default -1))
  (bind _self self)
  (bind _returnvalue (Minigame|Stats|GetMinigameStatsComponent _self))
  (bind _returnvalue_2 (>= ForceTeam 0))
  (Utilities|IsValid StatFilter
    (:Is%20Valid
      (switch Utilities|FlowControl|Switch|SwitchonE_StatManager_Scope (Variables|UserOptions|GetScope)
        (:NewEnumerator0
          (Variables|GetPlayerInstigator
            (:then
              (return (Minigame|Stats|GetStatforPlayerId _returnvalue (PlayerState|GetUniqueNetId (Variables|/Script/Engine.Controller|GetPlayerState (Variables|GetPlayerInstigator))) (Variables|GetStatFilter) false)))
            (:else)))
        (:NewEnumerator1
          (Variables|GetPlayerInstigator
            (:then
              _)
            (:else
              (if (not _returnvalue_2)
                (return 0))))
          (return (Minigame|Stats|GetStatforPlayerBucket _returnvalue (Math|Integer|SelectInt ForceTeam (Teams|TeamtoIndex (Team|GetActorTeam (Variables|GetPlayerInstigator))) _returnvalue_2) (Variables|GetStatFilter) false false)))
        (:NewEnumerator2
          (return (Minigame|Stats|GetStatforGroup _returnvalue (Variables|GetStatFilter) false)))))
    (:Is%20Not%20Valid
      (return 0))))

(fn GetScopeAsInt
  (return (Math|Conversions|ToInteger(Byte) (Variables|UserOptions|GetScope))))

(fn GetStatDetails
  (params (Stat%20Name :type Stat%20Event%20Filter%20Structure))
  (bind (_statevent _itemdefinition _targettagsquery) (Utilities|Struct|BreakStatEventFilter Stat%20Name))
  (bind (_staticon_smallicon _staticon_largeicon) (Utilities|Struct|BreakCreativeIconOption (Variables|GetStatIcon_SmallIcon)))
  (if (GameplayTags|HasTag _targettagsquery (Variables|Default|GetValueStatTag))
    (return true _staticon_largeicon (Variables|UserOptions|UI|GetStatColor))
    (else
      (return false "" 0))))

(fn GetTeamLevel
  (params (TeamIndex :type Integer))
  (return (Default|GetRelevantStatValue (Class|FortCreativeStatManagerBlueprintLibrary|GetStatFilter self (Variables|Default|GetLevelStatTag)) 0 TeamIndex)))

(fn GetTeamValue
  (params (TeamIndex :type Integer))
  (bind _valuestattag (Variables|Default|GetValueStatTag))
  (bind _self self)
  (bind _returnvalue (Class|FortCreativeStatManagerBlueprintLibrary|GetStatFilter _self _valuestattag))
  (bind _self_2 self)
  (bind _returnvalue_2 (Minigame|Stats|GetMinigameStatsComponent _self_2))
  ;[0 0.787 1 0.2] Returns the Stat Value for a specific Team. If we're using Match Scope, this will return the match value (since any team has the same score there). If we have a Player Scope, it'll always return 0.
  (switch Utilities|FlowControl|Switch|SwitchonE_StatManager_Scope (Variables|UserOptions|GetScope)
    (:NewEnumerator0
      (return 0))
    (:NewEnumerator1
      (return (Minigame|Stats|GetStatforPlayerBucket _returnvalue_2 TeamIndex _returnvalue false false)))
    (:NewEnumerator2
      (return (Minigame|Stats|GetStatforGroup _returnvalue_2 _returnvalue false)))))

(fn IgnoreWorldGridScaleOverride
  (return false))

(fn InitializeLevelforAllPlayers
  ;[1 1 1] If we have 0 level for a valid group, set it to 1, as all levels start at 1 and cannot go below that.
  (Utilities|HasLevels
    (:Yes
      (bind _currentminigame (Variables|/Script/FortniteGame.FortMinigameLogicComponent|GetCurrentMinigame (Variables|Default|GetFortMinigameProgress))
        (:then
          (LocalVariables|SetMinigame _currentminigame)
          (bind (_playerstates _teams _match) (Verse|GetAllMembersForGroup "NewEnumerator2"))
          (for _array_element _playerstates
            (CallFunction|InitializeLevelForPlayer _array_element))
          (for _array_element_2 _teams
            (if (== (Default|GetRelevantStatValue (Class|FortCreativeStatManagerBlueprintLibrary|GetStatFilter self (Variables|Default|GetLevelStatTag)) 0 (Teams|TeamtoIndex _array_element_2)) 0)
              (StatUpdating|ApplyLeveltoStatGroup 0 1 false false (Teams|TeamtoIndex _array_element_2))))
          (if _match
            (if (== (Minigame|Stats|GetStatforGroup (Minigame|Stats|GetMinigameStatsComponent self) (Class|FortCreativeStatManagerBlueprintLibrary|GetStatFilter self (Variables|Default|GetLevelStatTag)) false) 0)
              (StatUpdating|ApplyLeveltoStatGroup 0 1 false false -1 false true))))
        (:else)))))

(fn InitializeLevelforPlayer
  (params (PlayerState :type Player%20State%20Object%20Reference))
  (bind _returnvalue (Fortnite|GetControllerfromActor PlayerState))
  (Utilities|IsValid PlayerState
    (:Is%20Valid
      (Utilities|HasLevels
        (:Yes
          (if (== (Default|GetRelevantStatValue (Class|FortCreativeStatManagerBlueprintLibrary|GetStatFilter self (Variables|Default|GetLevelStatTag)) _returnvalue) 0)
            (StatUpdating|ApplyLeveltoStatGroup _returnvalue 1 false)))))))

(fn InitializePersistence
  (params (InLogicComponent :type Fort%20Minigame%20Logic%20Component%20Object%20Reference))
  (if (not (Editor|InEditor))
    (Persistence|InitializePersistenceComponent (Variables|Default|GetFortCreativePersistenceComponentBase) InLogicComponent true)
    (bind _spawnedobject (Fortnite|Utilities|SpawnTransientObject (Variables|Persistence|GetPersistenceData_StatManager_Class) self))
    (Variables|Persistence|SetPersistenceData_StatManager _spawnedobject)))

(fn LoadStat
  (params (PlayerInstigatorState :type Player%20State%20Object%20Reference) (OnlyIfEmpty :type Boolean))
  (bind _playerinstigatorstate (Variables|GetPlayerInstigatorState))
  (bind _returnvalue (PlayerState|GetUniqueNetId _playerinstigatorstate))
  (bind _levelstattag (Variables|Default|GetLevelStatTag))
  (bind _self self)
  (bind _returnvalue_2 (Class|FortCreativeStatManagerBlueprintLibrary|GetStatFilter _self _levelstattag))
  (bind _playerinstigatorstate_2 (Variables|GetPlayerInstigatorState))
  (bind _returnvalue_3 (PlayerState|GetPlayerController _playerinstigatorstate_2))
  (bind _playerinstigatorstate_3 (Variables|GetPlayerInstigatorState))
  (bind _returnvalue_4 (PlayerState|GetPlayerController _playerinstigatorstate_3))
  (bind _valuestattag (Variables|Default|GetValueStatTag))
  (bind _self_2 self)
  (bind _returnvalue_5 (Class|FortCreativeStatManagerBlueprintLibrary|GetStatFilter _self_2 _valuestattag))
  (bind _statvalue (Default|GetRelevantStatValue _returnvalue_5 _returnvalue_4))
  (bind _statvalue_2 (Default|GetRelevantStatValue _returnvalue_2 _returnvalue_4))
  (bind _persistencedata_statmanager (Variables|Persistence|GetPersistenceData_StatManager))
  (bind _statvalue_3 (Variables|/CRD_StatManager/SetupAssets/Persistence/PersistenceData_StatManager.PersistenceData_StatManager_C|GetStatValue _persistencedata_statmanager))
  (bind _persistencedata_statmanager_2 (Variables|Persistence|GetPersistenceData_StatManager))
  (bind _persistencedata_statmanager_3 (Variables|Persistence|GetPersistenceData_StatManager))
  (bind _statlevel (Variables|/CRD_StatManager/SetupAssets/Persistence/PersistenceData_StatManager.PersistenceData_StatManager_C|GetStatLevel _persistencedata_statmanager_3))
  (Utilities|IsValid PlayerInstigatorState
    (:Is%20Valid
      (if (and (Class|CreativeEnabledComponent|IsEnabled (Variables|Default|GetEnabledComponent)) (Variables|UserOptions|Persistence|GetUsePersistence))
        (Persistence|ResetPersistenceObject)
        (Variables|/CRD_StatManager/SetupAssets/Persistence/PersistenceData_StatManager.PersistenceData_StatManager_C|SetbDataIsLoaded false _persistencedata_statmanager_2)
        (Persistence|LoadData (Variables|Default|GetFortCreativePersistenceComponentBase) _returnvalue (Variables|Persistence|GetSavePathName) _persistencedata_statmanager_2)
        (if (Variables|/CRD_StatManager/SetupAssets/Persistence/PersistenceData_StatManager.PersistenceData_StatManager_C|GetbDataIsLoaded _persistencedata_statmanager_2)
          (LocalVariables|SetExistingStatValue _statvalue)
          (CallFunction|LogStatShipping _returnvalue _returnvalue_5 _statvalue false "LoadStat Value")
          (LocalVariables|SetExistingLevelValue _statvalue_2)
          (CallFunction|LogStatShipping _returnvalue _returnvalue_2 _statvalue_2 false "LoadStat Level")
          (if (Variables|GetOnlyIfEmpty)
            (if (and (== (LocalVariables|GetExistingStatValue) 0) (<= (LocalVariables|GetExistingLevelValue) 1))
              _))
          (if (>= _statlevel 0)
            (CallFunction|SetRelevantStatValue (Class|FortCreativeStatManagerBlueprintLibrary|GetStatFilter self (Variables|Default|GetLevelStatTag)) _statlevel _returnvalue_3)
            (else
              _))
          (if (>= _statvalue_3 0)
            (CallFunction|SetRelevantStatValue (Class|FortCreativeStatManagerBlueprintLibrary|GetStatFilter self (Variables|Default|GetValueStatTag)) _statvalue_3 _returnvalue_3)))))))

(fn LoadStatAll
  (params (FilterByRelevantToPS :type Player%20State%20Object%20Reference))
  (bind _playerstoload (LocalVariables|GetPlayersToLoad))
  (bind _output (Utilities|Array|Get(acopy) _playerstoload))
  (bind _returnvalue (PlayerState|GetPlayerController _output))
  (bind _newvalue (LocalVariables|GetNewValue))
  (bind _newlevel (LocalVariables|GetNewLevel))
  (Networking|SwitchHasAuthority
    (:Authority
      (if (Class|CreativeEnabledComponent|IsEnabled (Variables|Default|GetEnabledComponent))
        (bind _currentminigame (Variables|/Script/FortniteGame.FortMinigameLogicComponent|GetCurrentMinigame (Variables|Default|GetFortMinigameProgress))
          (:then
            (bind _output_get (LocalVariables|SetLocalMinigame _currentminigame))
            (Utilities|IsValid (Minigame|GetMinigameTeamComponent _output_get)
              (:Is%20Valid
                (LocalVariables|SetLocalMiniggameTeamComponent (Minigame|GetMinigameTeamComponent _output_get))
                (switch Utilities|FlowControl|Switch|SwitchonE_StatManager_Scope (Variables|UserOptions|GetScope)
                  (:NewEnumerator0
                    (bind _filterbyrelevanttops (Variables|GetFilterByRelevantToPS)
                      (:then
                        (Persistence|LoadStat _filterbyrelevanttops))
                      (:else
                        (bind _outplayers (Minigame|GetParticipatingPlayers_NotPure (LocalVariables|GetLocalMinigame)))
                        (for _array_element _outplayers
                          (Persistence|LoadStat _array_element)))))
                  (:NewEnumerator1
                    (bind _filterbyrelevanttops_2 (Variables|GetFilterByRelevantToPS)
                      (:then
                        (LocalVariables|SetPlayersToLoad (Class|FortMinigameTeamComponent|GetTeamMembers (LocalVariables|GetLocalMiniggameTeamComponent) (Team|GetActorPvPTeamIndex _filterbyrelevanttops_2))))
                      (:else
                        (for _array_element_2 (Variables|/Script/FortniteGame.FortMinigameTeamComponent|GetTeamArray (LocalVariables|GetLocalMiniggameTeamComponent))
                          (LocalVariables|SetPlayersToLoad (Class|FortMinigameTeamComponent|GetTeamMembers (LocalVariables|GetLocalMiniggameTeamComponent) (Utilities|Struct|BreakMinigameTeam _array_element_2)))
                          (bind (_value _level) (Persistence|GetBestValuestoLoad _playerstoload))
                          (LocalVariables|SetNewValue _value)
                          (LocalVariables|SetNewLevel _level)
                          (if (>= _newlevel 0)
                            (CallFunction|SetRelevantStatValue (Class|FortCreativeStatManagerBlueprintLibrary|GetStatFilter self (Variables|Default|GetLevelStatTag)) _newlevel _returnvalue)
                            (else
                              _))
                          (if (>= _newvalue 0)
                            (CallFunction|SetRelevantStatValue (Class|FortCreativeStatManagerBlueprintLibrary|GetStatFilter self (Variables|Default|GetValueStatTag)) _newvalue _returnvalue))))))
                  (:NewEnumerator2
                    (bind _outplayers_2 (Minigame|GetParticipatingPlayers_NotPure (LocalVariables|GetLocalMinigame)))
                    (LocalVariables|SetPlayersToLoad _outplayers_2))))))
          (:else))))))

(fn LogStatShipping
  (params (PlayerId :type Unique%20Net%20Id%20Structure) (StatFilter :type Fort%20Minigame%20Stat%20Filter%20Object%20Reference) (StatValue :type Integer) (IsWrite :type Boolean) (LogLocation :type String))
  (bind _returnvalue (PlayerInfo|GetDebugStringforUniqueId PlayerId))
  (Fortnite|FortShippingLog self (Utilities|String|Append (Utilities|String|SelectString "Writing " "Reading " IsWrite) " Player ID: " _returnvalue " Stat Filter: " (Utilities|GetDisplayName StatFilter) " Stat Value: " (Utilities|String|ToString(Integer) StatValue) " Location: " LogLocation) true))

(fn LogTags
  (params (ErrorMessage :type Text) (Tag :type Gameplay%20Tag%20Structure)))

(fn MakeUniqueStatName
  (bind _statname (Variables|UserOptions|GetStatName))
  (bind _returnvalue (Utilities|String|ToString(Text) _statname))
  (bind _statname_2 (Variables|UserOptions|GetStatName))
  (bind _returnvalue_2 (Utilities|String|ToString(Text) _statname_2))
  (bind _statname_3 (Variables|UserOptions|GetStatName))
  (bind _returnvalue_3 (Utilities|String|ToString(Text) _statname_3))
  (bind _level_stat_suffix (Variables|Default|GetLevel%20Stat%20Suffix))
  (bind _returnvalue_4 (Utilities|String|ToString(Text) _level_stat_suffix))
  (bind _returnvalue_5 (Utilities|String|FindSubstring _returnvalue _returnvalue_4 false true))
  (bind _valuestattag (Variables|Default|GetValueStatTag))
  (Networking|SwitchHasAuthority
    (:Authority
      (if (CallFunction|IsinPlayMode)
        ;[0 0.787 1 0.2] The following validation is going to be expensive (lots of array iteration to check things don't already exist), so if we're not in edit mode, just take the device's word that it's valid.
        (return false (Utilities|String|ToString(Text) (Variables|UserOptions|GetStatName)))
        (elif (Utilities|String|NotEqualExactly(String) (Variables|Cache|GetOldStatName) _returnvalue_3)
          (Utilities|FlowControl|Sequence
            (:then_0
              (bind _output_get (LocalVariables|SetFound%20Substring%20Index _returnvalue_5))
              (if (> _output_get 0)
                ;[0 0.5753 1] Iterating through the string from the end backwards. If we find a number at the end of the string, chop it off and save it into a separate string. Once we've chopped off all the numbers, put them in the correct order (since we were iterating through them backwards), convert them to a number, and use this as the starting point for the later search.
                (bind _output_get_2 (LocalVariables|SetNameAsString (Utilities|String|Left (Utilities|String|ToString(Text) (Variables|UserOptions|GetStatName)) _returnvalue_5)))
                (else
                  (LocalVariables|SetNameAsString (Utilities|String|ToString(Text) (Variables|UserOptions|GetStatName)))))
              (while (Utilities|String|IsNumeric (Utilities|String|Right _output_get_2 1))
                (LocalVariables|SetWasNumber true)
                (LocalVariables|SetNumberName (Utilities|String|Append (LocalVariables|GetNumberName) (Utilities|String|Right _output_get_2 1)))
                (LocalVariables|SetNameAsString (Utilities|String|LeftChop (LocalVariables|GetNameAsString) 1)))
              (if (LocalVariables|GetWasNumber)
                (LocalVariables|SetCountSuffix (Utilities|String|StringToInteger (Utilities|String|Reverse (LocalVariables|GetNumberName))))
                (LocalVariables|SetTempStatName (Utilities|String|Append (LocalVariables|GetNameAsString) (Utilities|String|ToString(Integer) (LocalVariables|GetCountSuffix))))
                (else
                  (LocalVariables|SetTempStatName (LocalVariables|GetNameAsString)))))
            (:then_1
              ;[0 0.787 1 0.2] Gather all of the stats in use.
              (bind _outstatevents (Class|FortCreativeStatManagerBlueprintLibrary|GetAllStatDisplayNames self))
              (if (.ReturnValue (Utilities|Map|Find _outstatevents _valuestattag))
                (if (Utilities|String|EqualExactly(String) (Utilities|String|ToString(Text) (Utilities|Map|Find _outstatevents _valuestattag)) _returnvalue_2)
                  (return false _returnvalue_2)
                  (else
                    _)))
              (bind _values (Utilities|Map|Values _outstatevents))
              (for _array_element _values
                (Utilities|Set|Add (LocalVariables|GetStatNames) (Utilities|String|ToString(Text) _array_element))))
            (:then_2
              ;[0 0.787 1 0.2] Iterate repeatedly until you find a number suffix that's not already taken.
              (while (Utilities|Set|ContainsItem (LocalVariables|GetStatNames) (LocalVariables|GetTempStatName))
                (LocalVariables|SetNeedsOverridden true)
                (LocalVariables|SetTempStatName (Utilities|String|Append (LocalVariables|GetNameAsString) (Utilities|String|ToString(Integer) (LocalVariables|GetCountSuffix))))
                (Math|Integer|IncrementInt (LocalVariables|GetCountSuffix)))
              ;[0 0.787 1 0.2] Return the completed name
              (return (LocalVariables|GetNeedsOverridden) (LocalVariables|GetTempStatName))))
          (else
            (return false _returnvalue_3)))))))

(fn OnFinishedRemovingMinigamePlayer
  (params (UniqueNetId :type Unique%20Net%20Id%20Structure) (bIsLocalPlayer :type Boolean))
  ;[1 1 1] If a player has left, remove them from the list of participating players.
  (Utilities|Array|RemoveItem (Variables|Default|GetActivePlayers) UniqueNetId))

(fn OnPlayerJoinInProgress
  (params (Pawn :type Pawn%20Object%20Reference))
  ;[0.04337 1 0.127 0.187] If we're joining in progress, load our stat, but only if the appropriate bucket doesn't already have a stat (Don't want to stomp on an in-progress stat).
  (if (and (Variables|UserOptions|Persistence|GetbAutoLoad) (Variables|UserOptions|Persistence|GetUsePersistence))
    (Persistence|LoadStat (Variables|/Script/Engine.Controller|GetPlayerState (Pawn|GetController Pawn)) true)))

(fn OnLastChanceToSave
  (params (UniqueNetId :type Unique%20Net%20Id%20Structure) (bIsLocalPlayer :type Boolean))
  (bind _returnvalue (Class|CreativeBlueprintLibrary|GetPlayerStatefromUniqueNetID UniqueNetId))
  (Networking|SwitchHasAuthority
    (:Authority
      (if (and (Variables|UserOptions|Persistence|GetbAutoSave) (Variables|UserOptions|Persistence|GetUsePersistence))
        (Utilities|IsValid _returnvalue
          (:Is%20Valid
            (Persistence|SaveStat _returnvalue)))))))

(fn OverrideStatNameProperty
  (params (Old%20Stat%20Name :type String))
  (bind _output_get (Variables|Cache|SetOldStatName Old%20Stat%20Name))
  (if (Utilities|String|NotEqualExactly(String) _output_get (Utilities|String|ToString(Text) (Variables|UserOptions|GetStatName)))
    (Utilities|Text|EditTextSourceString 0 (Variables|UserOptions|GetStatName) Old%20Stat%20Name)
    (OptionData|RefreshPropertyOverrides (Variables|Default|GetToyOptionsComponent))
    (return)))

(fn PopulatePersistenceDataObject
  (params (Controller :type Controller%20Object%20Reference))
  (bind _persistencedata_statmanager (Variables|Persistence|GetPersistenceData_StatManager))
  (bind _controller (Variables|GetController))
  (Persistence|ResetPersistenceObject)
  (Utilities|IsValid Controller
    (:Is%20Valid
      (bind _value (Verse|Value|GetPlayerValue _controller))
      (Variables|/CRD_StatManager/SetupAssets/Persistence/PersistenceData_StatManager.PersistenceData_StatManager_C|SetStatValue _value _persistencedata_statmanager)
      (bind _value_2 (Verse|Level|GetPlayerLevel _controller))
      (Variables|/CRD_StatManager/SetupAssets/Persistence/PersistenceData_StatManager.PersistenceData_StatManager_C|SetStatLevel _value_2 _persistencedata_statmanager)
      (Variables|/CRD_StatManager/SetupAssets/Persistence/PersistenceData_StatManager.PersistenceData_StatManager_C|SetbDataIsLoaded true _persistencedata_statmanager))))

(fn ReplaceWithOtherStatCreator
  (Utilities|FlowControl|Sequence
    (:then_0
      (bind (_outstatrow _returnvalue) (Class|FortCreativeStatManagerBlueprintLibrary|IsStatTaginUse self (Variables|Default|GetValueStatTag)))
      (if _returnvalue
        (DataRegistry|RemoveDataRegistryValueForUserOption (Utilities|Struct|MakeDataRegistryId "Creative.Stats" _outstatrow))))
    (:then_1
      (bind (_outstatrow_2 _returnvalue_2) (Class|FortCreativeStatManagerBlueprintLibrary|IsStatTaginUse self (Variables|Default|GetLevelStatTag)))
      (if _returnvalue_2
        (DataRegistry|RemoveDataRegistryValueForUserOption (Utilities|Struct|MakeDataRegistryId "Creative.Stats" _outstatrow_2))))
    (:then_2
      (CallFunction|SetupMinigameStat false)
      (Variables|Default|SetIsBeingReplaced true)
      (return (Variables|UserOptions|GetStatName) (Variables|Default|GetValueStatTag) (Variables|Default|GetLevelStatTag)))))

(fn ResetPersistenceObject
  (bind _persistencedata_statmanager (Variables|Persistence|GetPersistenceData_StatManager))
  (Variables|/CRD_StatManager/SetupAssets/Persistence/PersistenceData_StatManager.PersistenceData_StatManager_C|SetbDataIsLoaded false _persistencedata_statmanager)
  (Variables|/CRD_StatManager/SetupAssets/Persistence/PersistenceData_StatManager.PersistenceData_StatManager_C|SetStatLevel 0 _persistencedata_statmanager)
  (Variables|/CRD_StatManager/SetupAssets/Persistence/PersistenceData_StatManager.PersistenceData_StatManager_C|SetStatValue 0 _persistencedata_statmanager))

(fn RestorefromCache
  (bind _levelstattag (Variables|Default|GetLevelStatTag))
  (bind _self self)
  (bind _returnvalue (Class|FortCreativeStatManagerBlueprintLibrary|GetStatFilter _self _levelstattag))
  (bind _valuestattag (Variables|Default|GetValueStatTag))
  (bind _self_2 self)
  (bind _returnvalue_2 (Class|FortCreativeStatManagerBlueprintLibrary|GetStatFilter _self_2 _valuestattag))
  (bind _checknetid (LocalVariables|GetCheckNetID))
  (bind _cache_playervalue (Variables|StatCache|GetCache_PlayerValue))
  (bind _value (Utilities|Map|Find _cache_playervalue _checknetid))
  (bind _checknetid_2 (LocalVariables|GetCheckNetID))
  (bind _cache_playerlevel (Variables|StatCache|GetCache_PlayerLevel))
  (bind _value_2 (Utilities|Map|Find _cache_playerlevel _checknetid_2))
  (bind _checknetid_3 (LocalVariables|GetCheckNetID))
  (bind _output_get (LocalVariables|SetStatsComponent (Minigame|Stats|GetMinigameStatsComponent self)))
  (Utilities|IsValid _output_get
    (:Is%20Valid
      (switch Utilities|FlowControl|Switch|SwitchonE_StatManager_Scope (Variables|UserOptions|GetScope)
        (:NewEnumerator0
          (bind _keys (Utilities|Map|Keys (Variables|StatCache|GetCache_PlayerValue)))
          (for _array_element _keys
            (bind _output_get_2 (LocalVariables|SetCheckNetID _array_element))
            (Utilities|ValidforStat (Class|CreativeBlueprintLibrary|GetPlayerStatefromUniqueNetID _output_get_2)
              (:True
                (CallFunction|LogStatShipping _output_get_2 _returnvalue_2 _value true "Restore from Cache Value")
                (Minigame|Stats|SetStatforPlayerId (LocalVariables|GetStatsComponent) _output_get_2 _returnvalue_2 _value false)
                (Utilities|HasLevels
                  (:Yes
                    (CallFunction|LogStatShipping _checknetid_3 _returnvalue _value_2 true "Restore from Cache Level")
                    (Minigame|Stats|SetStatforPlayerId (LocalVariables|GetStatsComponent) _checknetid_3 _returnvalue _value_2 false))))) ; Only restore a player from the cache\nif they can currently support the stat\n(they may have changed teams since that occurred)
          ))
        (:NewEnumerator1
          (bind _keys_2 (Utilities|Map|Keys (Variables|StatCache|GetCache_TeamValue)))
          (for _array_element_2 _keys_2
            (bind _output_get_3 (LocalVariables|SetCheckTeamIndex _array_element_2))
            (Minigame|Stats|SetStatforPlayerBucket (LocalVariables|GetStatsComponent) _output_get_3 (Class|FortCreativeStatManagerBlueprintLibrary|GetStatFilter self (Variables|Default|GetValueStatTag)) (Utilities|Map|Find (Variables|StatCache|GetCache_TeamValue) (LocalVariables|GetCheckTeamIndex)) false)
            (Utilities|HasLevels
              (:Yes
                (Minigame|Stats|SetStatforPlayerBucket (LocalVariables|GetStatsComponent) (LocalVariables|GetCheckTeamIndex) (Class|FortCreativeStatManagerBlueprintLibrary|GetStatFilter self (Variables|Default|GetLevelStatTag)) (Utilities|Map|Find (Variables|StatCache|GetCache_TeamLevel) (LocalVariables|GetCheckTeamIndex)) false)))))
        (:NewEnumerator2
          (Minigame|Stats|SetStatforGroup (LocalVariables|GetStatsComponent) (Class|FortCreativeStatManagerBlueprintLibrary|GetStatFilter self (Variables|Default|GetValueStatTag)) (Variables|StatCache|GetCache_MatchValue) false)
          (Utilities|HasLevels
            (:Yes
              (Minigame|Stats|SetStatforGroup (LocalVariables|GetStatsComponent) (Class|FortCreativeStatManagerBlueprintLibrary|GetStatFilter self (Variables|Default|GetLevelStatTag)) (Variables|StatCache|GetCache_MatchLevel) false))))))))

(fn SaveStat
  (params (PlayerInstigatorState :type Player%20State%20Object%20Reference))
  (bind _playerinstigatorstate (Variables|GetPlayerInstigatorState))
  (bind _returnvalue (PlayerState|GetPlayerController _playerinstigatorstate))
  (bind _valuestattag (Variables|Default|GetValueStatTag))
  (bind _self self)
  (bind _returnvalue_2 (Class|FortCreativeStatManagerBlueprintLibrary|GetStatFilter _self _valuestattag))
  (bind _statvalue (Default|GetRelevantStatValue _returnvalue_2 _returnvalue))
  (bind _returnvalue_3 (PlayerState|GetUniqueNetId _playerinstigatorstate))
  (bind _playerinstigatorstate_2 (Variables|GetPlayerInstigatorState))
  (bind _returnvalue_4 (PlayerState|GetUniqueNetId _playerinstigatorstate_2))
  (bind _cache_playervalue (Variables|StatCache|GetCache_PlayerValue))
  (bind (_value _returnvalue_5) (Utilities|Map|Find _cache_playervalue _returnvalue_4))
  (bind _levelstattag (Variables|Default|GetLevelStatTag))
  (bind _self_2 self)
  (bind _returnvalue_6 (Class|FortCreativeStatManagerBlueprintLibrary|GetStatFilter _self_2 _levelstattag))
  (bind _statvalue_2 (Default|GetRelevantStatValue _returnvalue_6 _returnvalue))
  (bind _persistencedata_statmanager (Variables|Persistence|GetPersistenceData_StatManager))
  (bind _persistencedata_statmanager_2 (Variables|Persistence|GetPersistenceData_StatManager))
  (if (and (Class|CreativeEnabledComponent|IsEnabled (Variables|Default|GetEnabledComponent)) (Variables|UserOptions|Persistence|GetUsePersistence))
    (Utilities|FlowControl|Sequence
      (:then_0
        (Variables|/CRD_StatManager/SetupAssets/Persistence/PersistenceData_StatManager.PersistenceData_StatManager_C|SetbDataIsLoaded true _persistencedata_statmanager)
        (if (Utilities|Set|ContainsItem (Variables|Default|GetInvalidPlayers) (Variables|GetPlayerInstigatorState))
          ;[0 0.787 1 0.2] If we're an invalid player (e.g. one with an invalid team or class that used to have a valid team/class) pull from the cache instead, that'll contain the last good value.
          (if _returnvalue_5
            (Variables|/CRD_StatManager/SetupAssets/Persistence/PersistenceData_StatManager.PersistenceData_StatManager_C|SetStatValue _value _persistencedata_statmanager_2)
            (bind _output_get (Variables|/CRD_StatManager/SetupAssets/Persistence/PersistenceData_StatManager.PersistenceData_StatManager_C|SetStatLevel (Utilities|Map|Find (Variables|StatCache|GetCache_PlayerLevel) _returnvalue_4) _persistencedata_statmanager_2))
            (LocalVariables|SetData%20Changed true)
            (bind _returnvalue_7 (PlayerInfo|GetDebugStringforUniqueId _returnvalue_4))
            (Fortnite|FortShippingLog self (Utilities|String|Append "Saving cached stats for player: " _returnvalue_7 "value: " (Utilities|String|ToString(Integer) _value) "level: " (Utilities|String|ToString(Integer) _output_get)) true))
          (else
            (Utilities|ValidforStat (Variables|GetPlayerInstigatorState)
              (:True
                (CallFunction|LogStatShipping _returnvalue_3 _returnvalue_2 _statvalue false "SaveStat Value")
                (Variables|/CRD_StatManager/SetupAssets/Persistence/PersistenceData_StatManager.PersistenceData_StatManager_C|SetStatValue _statvalue _persistencedata_statmanager)
                (CallFunction|LogStatShipping _returnvalue_3 _returnvalue_6 _statvalue_2 false "SaveStat Level")
                (Variables|/CRD_StatManager/SetupAssets/Persistence/PersistenceData_StatManager.PersistenceData_StatManager_C|SetStatLevel _statvalue_2 _persistencedata_statmanager)
                (LocalVariables|SetData%20Changed true))))))
      (:then_1
        (if (LocalVariables|GetData%20Changed)
          (Persistence|SaveData (Variables|Default|GetFortCreativePersistenceComponentBase) (Variables|Persistence|GetPersistenceData_StatManager) (PlayerState|GetUniqueNetId PlayerInstigatorState) (Variables|Persistence|GetSavePathName) false) ; If any data has changed, then save it.
        )))))

(fn SaveStatAll
  (Networking|SwitchHasAuthority
    (:Authority
      (if (Class|CreativeEnabledComponent|IsEnabled (Variables|Default|GetEnabledComponent))
        (bind _currentminigame (Variables|/Script/FortniteGame.FortMinigameLogicComponent|GetCurrentMinigame (Variables|Default|GetFortMinigameProgress))
          (:then
            (bind _outplayers (Minigame|GetParticipatingPlayers_NotPure _currentminigame))
            (for _array_element _outplayers
              (Persistence|SaveStat _array_element)))
          (:else))))))

(fn SendStatEvent
  (params (NewStatValue :type Integer) (PlayerInstigator :type Controller%20Object%20Reference) (bForValue :type Boolean))
  (bind _asfort_player_controller (Utilities|Casting|CastToFortPlayerController PlayerInstigator))
  (Utilities|IsValid _asfort_player_controller
    (:Is%20Valid
      (bind _returnvalue (CreativeDevices|GetMinigameStatsComp _asfort_player_controller "Athena"))
      (Utilities|IsValid _returnvalue
        (:Is%20Valid
          (Class|FortMinigameStatsComponent|SendStatEventwithTags _returnvalue _asfort_player_controller (Variables|/Script/FortniteGame.FortMinigameStatFilter|GetStatEvent (Class|FortCreativeStatManagerBlueprintLibrary|GetStatFilter self (select (Variables|GetbForValue) (Variables|Default|GetValueStatTag) (Variables|Default|GetLevelStatTag)))) self (GameplayTags|MakeGameplayTagContainerfromTag (select (Variables|GetbForValue) (Variables|Default|GetValueStatTag) (Variables|Default|GetLevelStatTag))) (GameplayTags|MakeGameplayTagContainerfromTag) (GameplayTags|MakeGameplayTagContainerfromTag) (Variables|GetNewStatValue)))))))

(fn SetRelevantStatValue
  (params (StatFilter :type Fort%20Minigame%20Stat%20Filter%20Object%20Reference) (NewStatValue :type Integer) (PlayerInstigator :type Controller%20Object%20Reference) (Force%20Team :type Integer :default -1))
  (bind _playerinstigator (Variables|GetPlayerInstigator))
  (bind _playerstate (Variables|/Script/Engine.Controller|GetPlayerState _playerinstigator))
  (bind _returnvalue (PlayerState|GetUniqueNetId _playerstate))
  (bind _sc (LocalVariables|GetSC))
  (bind _sc_2 (LocalVariables|GetSC))
  (Utilities|IsValid StatFilter
    (:Is%20Valid
      (if (Class|CreativeEnabledComponent|IsEnabled (Variables|Default|GetEnabledComponent))
        (Networking|SwitchHasAuthority
          (:Authority
            (LocalVariables|SetSC (Minigame|Stats|GetMinigameStatsComponent self))
            (LocalVariables|SetStatValueDelta (- (Variables|GetNewStatValue) (Default|GetRelevantStatValue (Variables|GetStatFilter) (Variables|GetPlayerInstigator) (Variables|GetForce%20Team))))
            ;[1 1 1] Adjust Relevant Stat Value
            (switch Utilities|FlowControl|Switch|SwitchonE_StatManager_Scope (Variables|UserOptions|GetScope)
              (:NewEnumerator0
                (Utilities|ValidforStat _playerinstigator
                  (:True
                    (CallFunction|LogStatShipping _returnvalue (Variables|GetStatFilter) (Variables|GetNewStatValue) true "Set Relevant Stat Value")
                    (Minigame|Stats|SetStatforPlayerId _sc_2 _returnvalue (Variables|GetStatFilter) (Variables|GetNewStatValue) false)
                    (bind (_is_valid _bucket_index) (CallFunction|GetBucketIndexforActor (Variables|GetPlayerInstigator) -1))
                    (if _is_valid
                      (Minigame|Stats|AddStattoPlayerBucket _sc_2 _bucket_index (Variables|GetStatFilter) (LocalVariables|GetStatValueDelta))
                      (else
                        _))
                    (Minigame|Stats|AddStattoGroup _sc_2 (Variables|GetStatFilter) (LocalVariables|GetStatValueDelta)))))
              (:NewEnumerator1
                (bind (_is_valid_2 _bucket_index_2) (CallFunction|GetBucketIndexforActor (Variables|GetPlayerInstigator) (Variables|GetForce%20Team)))
                (if _is_valid_2
                  (Minigame|Stats|SetStatforPlayerBucket _sc _bucket_index_2 (Variables|GetStatFilter) (Variables|GetNewStatValue) false)
                  (else
                    _))
                (Minigame|Stats|AddStattoGroup _sc (Variables|GetStatFilter) (LocalVariables|GetStatValueDelta)))
              (:NewEnumerator2
                (Minigame|Stats|SetStatforGroup (LocalVariables|GetSC) (Variables|GetStatFilter) (Variables|GetNewStatValue) false)))))))))

(fn SetAllLevel
  (params (Value :type Integer))
  (Utilities|HasLevels
    (:Yes
      (bind (_playerstates _teams) (Verse|GetAllMembersForGroup "NewEnumerator2"))
      (for _array_element _playerstates
        (StatUpdating|ApplyLeveltoStatGroup (PlayerState|GetPlayerController _array_element) Value false true))
      (for _array_element_2 _teams
        (StatUpdating|ApplyLeveltoStatGroup 0 Value false true (Teams|TeamtoIndex _array_element_2) true))
      (StatUpdating|ApplyLeveltoStatGroup 0 Value false true -1 true true))))

(fn SetAllValue
  (params (Value :type Integer))
  (bind (_playerstates _teams _match) (Verse|GetAllMembersForGroup "NewEnumerator2"))
  (for _array_element _playerstates
    (CallFunction|ApplyValueToStatGroup (PlayerState|GetPlayerController _array_element) (Variables|GetValue)))
  (for _array_element_2 _teams
    (CallFunction|ApplyValueToStatGroup 0 (Variables|GetValue) false (Teams|TeamtoIndex _array_element_2)))
  (if _match
    (CallFunction|ApplyValueToStatGroup 0 (Variables|GetValue) false -1 true)))

(fn SetPlayerLevel
  (params (Controller :type Controller%20Object%20Reference) (Value :type Integer))
  (StatUpdating|ApplyLeveltoStatGroup Controller Value false true -1 true))

(fn SetPlayerValue
  (params (Controller :type Controller%20Object%20Reference) (Value :type Integer))
  (bind (_playerstates _teams _match) (Verse|GetAllMembersForGroup "NewEnumerator0" Controller))
  (for _array_element _playerstates
    (CallFunction|ApplyValueToStatGroup (PlayerState|GetPlayerController _array_element) (Variables|GetValue)))
  (for _array_element_2 _teams
    (CallFunction|ApplyValueToStatGroup 0 (Variables|GetValue) false (Teams|TeamtoIndex _array_element_2)))
  (if _match
    (CallFunction|ApplyValueToStatGroup 0 (Variables|GetValue))))

(fn SetTeamLevel
  (params (TeamIndex :type Integer) (Value :type Integer))
  (Utilities|HasLevels
    (:Yes
      (bind (_playerstates _teams) (Verse|GetAllMembersForGroup "NewEnumerator1" 0 TeamIndex))
      (for _array_element _playerstates
        (StatUpdating|ApplyLeveltoStatGroup (PlayerState|GetPlayerController _array_element) Value false true -1 true))
      (for _array_element_2 _teams
        (StatUpdating|ApplyLeveltoStatGroup 0 Value false true (Teams|TeamtoIndex _array_element_2) true))
      (StatUpdating|ApplyLeveltoStatGroup 0 Value false true (Variables|Team|GetTeamIndex) true))))

(fn SetTeamValue
  (params (TeamIndex :type Integer) (Value :type Integer))
  (bind (_playerstates _teams _match) (Verse|GetAllMembersForGroup "NewEnumerator1" 0 TeamIndex))
  (for _array_element _playerstates
    (CallFunction|ApplyValueToStatGroup (PlayerState|GetPlayerController _array_element) (Variables|GetValue)))
  (for _array_element_2 _teams
    (CallFunction|ApplyValueToStatGroup 0 (Variables|GetValue) false (Teams|TeamtoIndex _array_element_2)))
  (if _match
    (CallFunction|ApplyValueToStatGroup 0 (Variables|GetValue))))

(fn SetupMinigameStat
  (params (Add :type Boolean))
  ;[0 0.787 1 0.2] Either create the Stat if we want it (or update it if it already exists) and bind events, or remove the player defined stat.
  (Networking|SwitchHasAuthority
    (:Authority
      (if Add
        (Utilities|MinigameHasStatSlots
          (:Yes
            (Utilities|FlowControl|Sequence
              (:then_0
                (Initialization|SetupOrUpdateStat (Variables|UserOptions|GetStatName) true)
                (Utilities|HasLevels
                  (:Yes
                    (Initialization|SetupOrUpdateStat (Utilities|Text|FormatText "NSLOCTEXT(\"[F72196ED6FAE244A270004C9785A108E]\", \"313BA0FD49FB6C3E8E63F19930C16A1A\", \"{statname}{levelsuffix}\")" (Variables|UserOptions|GetStatName) (Variables|Default|GetLevel%20Stat%20Suffix))))
                  (:No
                    (Utilities|IsValidStatTag (Variables|Default|GetLevelStatTag)
                      (:Valid
                        (Class|FortCreativeStatManagerBlueprintLibrary|RemovePlayerDefinedStat "Level" self)
                        (Variables|Default|SetLevelStatTag))))))
              (:then_1
                (Initialization|BindStatEvents true)))))
        (elif (not (Variables|Default|GetIsBeingReplaced))
          (Utilities|IsValidStatTag (Variables|Default|GetValueStatTag)
            (:Valid
              (Class|FortCreativeStatManagerBlueprintLibrary|RemovePlayerDefinedStat "Value" self))
            (:Not%20Valid
              _))
          (Utilities|IsValidStatTag (Variables|Default|GetLevelStatTag)
            (:Valid
              (Class|FortCreativeStatManagerBlueprintLibrary|RemovePlayerDefinedStat "Level" self))
            (:Not%20Valid
              _))
          (Initialization|BindStatEvents false)
          (Cache|ClearCaches)
          (else
            _))))))

(fn SetuporUpdateStat
  (params (StatName :type Text%20%28by%20ref%29) (bForValue :type Boolean))
  ;[0 0.787 1 0.2] If we don't have a filter yet, the stat hasn't been created, so set it up. If we do, update it.
  ; (There's an assert if you try and setup an already setup stat)
  (Networking|SwitchHasAuthority
    (:Authority
      (CallFunction|SetuporUpdateStatBasedonCurrentTags (select (Variables|GetbForValue) (Variables|Default|GetValueStatTag) (Variables|Default|GetLevelStatTag)) (select (Variables|GetbForValue) "Value" "Level") StatName))))

(fn SetupOrUpdateStatBasedOnCurrentTags
  (params (In%20Out%20Stat%20Tag :type Gameplay%20Tag%20Structure%20%28by%20ref%29) (Additional%20Tag%20Name :type String) (%20Name :type Text))
  (bind (_staticon_smallicon _staticon_largeicon) (Utilities|Struct|BreakCreativeIconOption (Variables|GetStatIcon_SmallIcon)))
  (bind (_staticon_smallicon_2 _staticon_largeicon_2) (Utilities|Struct|BreakCreativeIconOption (Variables|GetStatIcon_SmallIcon)))
  (bind _in_out_stat_tag (Variables|GetIn%20Out%20Stat%20Tag))
  (bind (_staticon_smallicon_3 _staticon_largeicon_3) (Utilities|Struct|BreakCreativeIconOption (Variables|GetStatIcon_SmallIcon)))
  (bind (_staticon_smallicon_4 _staticon_largeicon_4) (Utilities|Struct|BreakCreativeIconOption (Variables|GetStatIcon_SmallIcon)))
  (CallFunction|LogTags "NSLOCTEXT(\"[F72196ED6FAE244A270004C9785A108E]\", \"E251CEB54BAEF60777F53C9CDA2641E9\", \"Starting SetupOrUpdateBasedOnCurrentTags\")" (Variables|GetIn%20Out%20Stat%20Tag))
  (if (GameplayTags|IsGameplayTagValid (Variables|GetIn%20Out%20Stat%20Tag))
    (bind (_outstatrow _returnvalue) (Class|FortCreativeStatManagerBlueprintLibrary|IsStatTaginUse self (Variables|GetIn%20Out%20Stat%20Tag)))
    (if _returnvalue
      (CallFunction|LogTags "NSLOCTEXT(\"[F72196ED6FAE244A270004C9785A108E]\", \"EE95525849B59D0C5BC07FB4DC26E65F\", \"We\\'re using a tag other than loaded stat tag, let\\'s see if we can use our loaded one.\")" (Variables|GetIn%20Out%20Stat%20Tag))
      (bind _outgameplaytag (Class|FortCreativeStatManagerBlueprintLibrary|GetinUseStatTagfromRowName self (Class|FortCreativeStatManagerBlueprintLibrary|GetSpatialInterfaceScopedStatTagName self (Variables|GetAdditional%20Tag%20Name))))
      (bind _output_get (LocalVariables|SetInUseStatTag _outgameplaytag))
      (if (== _output_get (Variables|GetIn%20Out%20Stat%20Tag))
        (CallFunction|LogTags "NSLOCTEXT(\"[F72196ED6FAE244A270004C9785A108E]\", \"C14889BD4204CDEABFA15F8DB5E5DEDA\", \"Loaded stat tag was in use by us. Updating our data registry.\")" (Variables|GetIn%20Out%20Stat%20Tag))
        (Class|FortCreativeStatManagerBlueprintLibrary|UpdateStatProperties (Variables|Get%20Name) (Variables|GetIn%20Out%20Stat%20Tag) (Variables|GetAdditional%20Tag%20Name) _staticon_largeicon self)
        (else
          (Variables|SetBy-RefVar (Variables|GetIn%20Out%20Stat%20Tag) (LocalVariables|GetInUseStatTag))
          (CallFunction|LogTags "NSLOCTEXT(\"[F72196ED6FAE244A270004C9785A108E]\", \"27C8A22841990370C2A91AB00E1E2BE9\", \"Loaded stat tag was in use Switching back to old tag. Updating our data registry entry.\")" (Variables|GetIn%20Out%20Stat%20Tag))
          (Class|FortCreativeStatManagerBlueprintLibrary|UpdateStatProperties (Variables|Get%20Name) (Variables|GetIn%20Out%20Stat%20Tag) (Variables|GetAdditional%20Tag%20Name) _staticon_largeicon_3 self)))
      (else
        (CallFunction|LogTags "NSLOCTEXT(\"[F72196ED6FAE244A270004C9785A108E]\", \"BD1C1EFE4229067BBA1B19A42E120D0C\", \"Loaded Stat Tag not in use. Updating to use it.\")" (Variables|GetIn%20Out%20Stat%20Tag))
        (CallFunction|Resave)
        (Class|FortCreativeStatManagerBlueprintLibrary|UpdateStatProperties (Variables|Get%20Name) (Variables|GetIn%20Out%20Stat%20Tag) (Variables|GetAdditional%20Tag%20Name) _staticon_largeicon_4 self)))
    (else
      (Variables|SetBy-RefVar _in_out_stat_tag (Class|FortCreativeStatManagerBlueprintLibrary|GetNextPlayerDefinedStatTag self))
      (CallFunction|LogTags "NSLOCTEXT(\"[F72196ED6FAE244A270004C9785A108E]\", \"B9B44173467CBB94C619F8957F966EB9\", \"No Stat Tag\")" _in_out_stat_tag)
      (CallFunction|Resave)
      (Class|FortCreativeStatManagerBlueprintLibrary|SetupPlayerDefinedStat (Variables|Get%20Name) _in_out_stat_tag (Variables|GetAdditional%20Tag%20Name) _staticon_largeicon_2 self))))

(fn ShouldCallOnPreviewActors
  (return false))

(fn UpdateVisualDetailsfromServer
  (bind (_staticon_smallicon _staticon_largeicon) (Utilities|Struct|BreakCreativeIconOption (Variables|GetStatIcon_SmallIcon)))
  (bind (_staticon_smallicon_2 _staticon_largeicon_2) (Utilities|Struct|BreakCreativeIconOption (Variables|GetStatIcon_SmallIcon)))
  ;[0 0.787 1 0.2] In order to not require this actor to be always replicated on the Client, we instead funnel all of the data through a playspace component that we create here. This will cache all of our details and replicate it for the client to display.
  (if (not (Editor|InEditor))
    (Networking|SwitchHasAuthority
      (:Authority
        (bind _returnvalue (GUID|GenerateGuidfromObject self true))
        (LocalVariables|SetGUID _returnvalue)
        (if (Utilities|Enum|NotEqual(Enum) (Variables|UserOptions|UI|GetStatBarShownOnHUD) "NewEnumerator2")
          (bind _statdetailscomponent (UI|FindorCreatePlayspaceComponent))
          (Utilities|IsValid _statdetailscomponent
            (:Is%20Valid
              (Class|PlayspaceComponentPlayerStatDetails|RegisterorUpdateDevice _statdetailscomponent (LocalVariables|GetGUID) (Utilities|Struct|MakeStatCreatorDetails (Class|FortCreativeStatManagerBlueprintLibrary|GetStatFilter self (Variables|Default|GetValueStatTag)) (Class|FortCreativeStatManagerBlueprintLibrary|GetStatFilter self (Variables|Default|GetLevelStatTag)) (Variables|UserOptions|GetStatName) _staticon_largeicon (Variables|UserOptions|GetScope) (Variables|UserOptions|UI|GetStatColor) (Variables|Cache|GetExpNeededPerLevel) (LevelScaling|GetLevelNumber) (Variables|UserOptions|UI|GetUser%20Widget) (Variables|UserOptions|UI|GetHUD%20Priority) (Variables|UserOptions|UI|GetShowWidgetInList) (Utilities|Select :Index (Variables|UserOptions|UI|GetStatBarShownOnHUD) :NewEnumerator0 -1.0 :NewEnumerator1 0.0 :NewEnumerator2 0.0 :NewEnumerator3 (Variables|UserOptions|UI|GetStatBarDuration)) (Variables|UserOptions|Requirements|GetSelected%20Team) (Variables|UserOptions|Requirements|GetInvert%20Team%20Selection) (Variables|UserOptions|Requirements|GetSelected%20Class) (Variables|UserOptions|Requirements|GetInvert%20Class%20Selection) (Variables|UserOptions|UI|GetIntroAnimation) (Variables|UserOptions|UI|GetOutroAnimation)))
              (bind _currentminigame (Variables|/Script/FortniteGame.FortMinigameLogicComponent|GetCurrentMinigame (Variables|Default|GetFortMinigameProgress))
                (:then
                  (Class|PlayspaceComponentPlayerStatDetails|SetMinigame _statdetailscomponent _currentminigame))
                (:else))))
          (else
            (bind _statdetailscomponent_2 (Variables|Default|GetStatDetailsComponent)
              (:then
                (Default|DeregisterDevice _statdetailscomponent_2 (LocalVariables|GetGUID)))
              (:else))))))
    (else
      ;;[0 0.787 1 0.2] In editor stat details need to be manually propagated as we don't have access to the Player Controller component. These are all of our Stat Powerup actors
      (bind _outactors (Actor|GetAllActorswithInterface "/CRD_StatManager/BPi_AffectedByStat.BPi_AffectedByStat_C"))
      (for _array_element _outactors
        (StatDetails|UpdateStat(Message) _array_element (Variables|Default|GetValueStatTag) true _staticon_largeicon_2 (Variables|UserOptions|UI|GetStatColor) (Variables|UserOptions|GetStatName))))))

(fn UpdateFromOtherStatCreator
  (params (Actor :type Object%20Reference))
  (bind _asdevice_stat_creator (Utilities|Casting|CastToDevice_StatCreator Actor))
  (bind (_statname _valuestattag _levelstattag) (Replacement|ReplacewithOtherStatCreator _asdevice_stat_creator))
  (Utilities|Text|EditTextSourceString 0 (Variables|UserOptions|GetStatName) (Utilities|String|ToString(Text) _statname))
  (Variables|Cache|SetOldStatName (Utilities|String|ToString(Text) _statname))
  (if (!= (Variables|Default|GetValueStatTag) _valuestattag)
    (CallFunction|SetupMinigameStat false) ; Throwaway our generated stat if any,\nWe'll be stealing the one from the original device.
    (Variables|Default|SetValueStatTag _valuestattag)
    (Variables|Default|SetLevelStatTag _levelstattag)
    (CallFunction|SetupMinigameStat true)
    (UI|UpdateVisualDetailsFromServer)
    (else
      _))
  (Initialization|BindStatEvents true))

(fn ConstructionScript
  (|CallParentFunction))

(fn ValidatePlayerAndCacheorRestoreStats
  (params (PlayerState :type Player%20State%20Object%20Reference))
  (bind _netid (LocalVariables|GetNetID))
  (bind _cache_playervalue (Variables|StatCache|GetCache_PlayerValue))
  (bind (_value _returnvalue) (Utilities|Map|Find _cache_playervalue _netid))
  (bind _netid_2 (LocalVariables|GetNetID))
  (bind _cache_playerlevel (Variables|StatCache|GetCache_PlayerLevel))
  (bind _value_2 (Utilities|Map|Find _cache_playerlevel _netid_2))
  (bind _levelstattag (Variables|Default|GetLevelStatTag))
  (bind _self self)
  (bind _returnvalue_2 (Class|FortCreativeStatManagerBlueprintLibrary|GetStatFilter _self _levelstattag))
  (bind _valuestattag (Variables|Default|GetValueStatTag))
  (bind _self_2 self)
  (bind _returnvalue_3 (Class|FortCreativeStatManagerBlueprintLibrary|GetStatFilter _self_2 _valuestattag))
  (bind _netid_3 (LocalVariables|GetNetID))
  (bind _netid_4 (LocalVariables|GetNetID))
  (bind _playerstate (Variables|GetPlayerState))
  (bind _playerstate_2 (Variables|GetPlayerState))
  (bind _playerstate_3 (Variables|GetPlayerState))
  ;[0 0.787 1 0.2] If we're player scoped, either drop the player stat into the cache, or pull it from the cache and apply it, based on whether or not we are a valid actor.
  (bind _returnvalue_4 (Development|GetConsoleVariableBoolValue "FortCreativeStatManager.EnableJoinInProgressValidateGate"))
  (bind _asfort_player_state (Utilities|Casting|CastToFortPlayerState PlayerState)
    (:then
      (Utilities|IsValid (Class|FortPlayerState|FindFNEPlayerComponent _asfort_player_state)
        (:Is%20Valid
          (if (and (Minigame|IsJoiningGame (Class|FortPlayerState|FindFNEPlayerComponent _asfort_player_state)) _returnvalue_4)
            _
            (else
              _)))
        (:Is%20Not%20Valid
          _)))
    (:CastFailed
      _))
  (Utilities|IsEnabled
    (:true
      (if (Utilities|Enum|Equal(Enum) (Variables|UserOptions|GetScope) "NewEnumerator0")
        (LocalVariables|SetNetID (PlayerState|GetUniqueNetId _playerstate_2))
        (Utilities|ValidforStat _playerstate_2
          (:True
            (if (Utilities|Set|ContainsItem (Variables|Default|GetInvalidPlayers) _playerstate_3)
              (Utilities|Set|Remove (Variables|Default|GetInvalidPlayers) _playerstate_3)
              (Fortnite|FortShippingLog self (Utilities|String|Append "Removed Player from Invalid players: " (Utilities|GetDisplayName _playerstate_3)) true)
              (bind _output_get (LocalVariables|SetStatsComponent (Minigame|Stats|GetMinigameStatsComponent self)))
              (Utilities|IsValid _output_get
                (:Is%20Valid
                  (if _returnvalue
                    (CallFunction|LogStatShipping _netid_4 _returnvalue_3 _value true "Validate Player Value")
                    (Minigame|Stats|SetStatforPlayerId (LocalVariables|GetStatsComponent) _netid_4 _returnvalue_3 _value false)
                    (else
                      _))
                  (Utilities|HasLevels
                    (:Yes
                      (CallFunction|LogStatShipping _netid_3 _returnvalue_2 _value_2 true "Validate Player Level")
                      (Minigame|Stats|SetStatforPlayerId (LocalVariables|GetStatsComponent) _netid_3 _returnvalue_2 _value_2 false)))))))
          (:False
            (if (not (Utilities|Set|ContainsItem (Variables|Default|GetInvalidPlayers) _playerstate_2))
              (Utilities|Set|Add (Variables|Default|GetInvalidPlayers) _playerstate)
              (Fortnite|FortShippingLog self (Utilities|String|Append "Adding Player to Invalid players: " (Utilities|GetDisplayName _playerstate)) true)
              (Cache|CachePlayerStat (LocalVariables|GetNetID) (Variables|StatCache|GetCache_PlayerValue) (Class|FortCreativeStatManagerBlueprintLibrary|GetStatFilter self (Variables|Default|GetValueStatTag)))
              (Utilities|HasLevels
                (:Yes
                  (Cache|CachePlayerStat (LocalVariables|GetNetID) (Variables|StatCache|GetCache_PlayerLevel) (Class|FortCreativeStatManagerBlueprintLibrary|GetStatFilter self (Variables|Default|GetLevelStatTag))))))))))))

(fn ValidateStatName
  ;[0.1051 0.4063 1 0] Verifies the Stat name is unique within the minigame. If not, it'll attempt to create a new unique Stat Name, then updates the save data. This is expensive, so we're ensuring that it only runs when the stat name changes.
  (Networking|SwitchHasAuthority
    (:Authority
      (bind (_overridden _uniquename) (Initialization|MakeUniqueStatName))
      (if _overridden
        (Initialization|OverrideStatNameProperty _uniquename)
        (else
          (Variables|Cache|SetOldStatName (Utilities|String|ToString(Text) (Variables|UserOptions|GetStatName))))))))
