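# M_StatCounter_Background is a translucent, two-sided material in CRD_StatManager, named as the background for a stat counter.
# It uses the Default Lit shading model and contains a single Color vector parameter, defaulting to half-transparent black.
# The graph shows no connection from that parameter to any material input.
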Material /CRD_StatManager/Materials/M_StatCounter_Background
graph {
 properties {bAutomaticallySetUsageInEditor=false, blendMode=BLEND_Translucent, twoSided=true, usedShadingModels=Default Lit}
 nodes [VectorParameter() {parameterName=Color, defaultValue=<rgba 0 0 0 0.5>}]
}
