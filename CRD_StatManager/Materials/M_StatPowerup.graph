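# Lit material for a stat powerup pickup, with a tintable ring-masked color and a logo overlay.
# A switchable time-driven world position rotation spins the mesh about Z, with matching normals read from UV channels 2–3 and optionally blended with a normal map.
# A curve-atlas gradient, driven by Fresnel and object height, shades the color.
# A view-space sphere-map reflection and sun/moon-direction highlight add fake shine, masked by the packed spec texture.
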Material /CRD_StatManager/Materials/M_StatPowerup
graph {
 properties {
  allowTranslucentCustomDepthWrites=true
  bEnableResponsiveAA=true
  bUsedWithMeshDeformer=true
  bUsedWithNiagaraMeshParticles=true
  bUsedWithNiagaraSprites=true
  bUsedWithSkeletalMesh=true
  numCustomizedUVs=4
  opacityMaskClipValue=0
  pixelDepthOffsetMode=PDOM_Legacy
  usedShadingModels=Default Lit
 }
 nodes [
  n1 = ObjectOrientation()
  n2 = Multiply(A=Time(), B=ScalarParameter() {defaultValue=-0.2, group=MeshRotation, parameterName=RotationSpeed})
  n3 = Transform(None=ReflectionVectorWS(CustomWorldNormal=MaterialFunctionCall(Rotation Axis=@n1, Rotation Angle=@n2) {materialFunction=&FixRotateAboutAxisNormals}.New World Space Vertex Normal)) {transformSourceType=TRANSFORMSOURCE_World, transformType=TRANSFORM_View}
  n4 = ReflectionVectorWS()
  n5 = ComponentMask(None=StaticSwitchParameter(True=Multiply(A=@n3) {constB=-1}, False=Transform(None=@n4) {transformSourceType=TRANSFORMSOURCE_World, transformType=TRANSFORM_View}) {defaultValue=true, group=MeshRotation, parameterName=RotateAroundAxis}) {g=true, r=true}
  n6 = ScalarParameter() {defaultValue=1, group=Reflection, parameterName=ReflectionMapTilingU, sortPriority=5}
  n7 = ScalarParameter() {defaultValue=1, group=Reflection, parameterName=ReflectionMapTilingV, sortPriority=5}
  n8 = TextureSampleParameter2D(UVs=Multiply(A=ConstantBiasScale(None=@n5), B=AppendVector(A=@n6, B=@n7))) {
   group=Reflection
   parameterName=SphereReflectionMap
   sortPriority=1
   texture=/Game/Creative/Devices/Generic/Materials/Textures/T_CP_Device_Reflection
  }
  n9 = Multiply(
   A=Power(Base=@n8.RGB, Exp=ScalarParameter() {defaultValue=1, group=Reflection, parameterName=ReflectionExponent})
   B=VectorParameter() {group=Reflection, parameterName=Reflection Tint, sortPriority=2, defaultValue=<rgba 1 1 1 1>}.RGB
  )
  n10 = ScalarParameter() {defaultValue=0.25, group=Reflection, parameterName=Reflection Multiplier}
  n11 = CollectionParameter() {
   collection=&FortniteMaterialParameters
   parameterName=SunAndMoonModelDirectionalVector
  }
  n12 = ComponentMask(None=@n11) {b=true, g=true, r=true}
  n13 = ReflectionVectorWS()
  n14 = Power(Base=Clamp(None=DotProduct(A=@n12, B=@n13)), Exp=ScalarParameter() {defaultValue=200, group=Specular, parameterName=Spec Highlight Power}) {constExponent=20}
  n15 = ScalarParameter() {defaultValue=2, group=Specular, parameterName=Spec Intense}
  n16 = TextureSampleParameter2D() {
   group=Texture
   parameterName=SpecMetalRoughAO Map
   texture=/Game/Creative/Devices/Generic/Materials/Textures/T_CP_Device_GearIcon_Placed_01_S
  }
  n17 = LinearInterpolate(B=@n16.R, Alpha=ScalarParameter() {defaultValue=1, group=Specular, parameterName=ReflectSpec Mask Influence}) {constA=1}
  Multiply(A=Add(A=Multiply(A=@n9, B=@n10), B=Multiply(A=@n14, B=@n15)), B=@n17) {constB=2}
  n18 = ComponentMask(None=Subtract(A=WorldPosition().XYZ, B=ObjectPositionWS())) {b=true}
  n19 = ObjectRadius()
  n20 = Saturate(None=ConstantBiasScale(None=Divide(A=@n18, B=@n19)))
  n21 = LinearInterpolate(A=@n20, B=OneMinus(None=@n20), Alpha=ScalarParameter() {group=Color Curve, parameterName=FlipGradient})
  n22 = ScalarParameter() {defaultValue=1.2, group=Color Curve, parameterName=VerticalGradientInflience}
  n23 = ScalarParameter() {defaultValue=1, group=Color Curve, parameterName=FresnelExp}
  n24 = ScalarParameter() {defaultValue=0.5, group=Color Curve, parameterName=FresnelMin}
  n25 = LinearInterpolate(Alpha=Fresnel(ExponentIn=ScalarParameter() {defaultValue=1, group=Color Curve, parameterName=EdgeFresnelXP}) {baseReflectFraction=0, exponent=1}) {constA=5, constB=-0.4}
  n26 = Power(Base=Multiply(A=Fresnel(ExponentIn=@n23, BaseReflectFractionIn=@n24) {baseReflectFraction=0, exponent=0.5}, B=Saturate(None=@n25)))
  n27 = CurveAtlasRowParameter(CurveTime=Saturate(None=LinearInterpolate(A=@n26, B=@n21, Alpha=@n22))) {
   atlas=&AthenaCurveAtlas01
   curve=/Game/Characters/Enemies/DUDEBRO/FX/Curve_StormKing_Alpha
   defaultValue=141
   desc=Curve Atlas
   group=Color Curve
   parameterName=CurveAtlas
   sortPriority=2
  }
  n28 = TextureCoordinate() {coordinateIndex=1}
  n29 = ScalarParameter() {defaultValue=0.5, group=Logo, parameterName=Logo Scale}
  n30 = ScalarParameter() {group=Logo, parameterName=Logo Offset U}
  n31 = ScalarParameter() {group=Logo, parameterName=Logo Offset V}
  n32 = TextureSampleParameter2D(UVs=Add(
    A=MaterialFunctionCall(UVs=@n28, Texture Scale=@n29) {materialFunction=&ScaleUVsByCenter}.UVs
    B=AppendVector(A=@n30, B=@n31)
   )) {
   group=Logo
   parameterName=Logo
   samplerSource=SSM_Clamp_WorldGroupSettings
   texture=/Game/Global/Textures/Default/Blanks/T_Blank_LegalWhite_sRGBColor
  }
  n33 = VectorParameter() {group=Color, parameterName=Color, sortPriority=14, defaultValue=<rgba 0.3125 0.2914 0.2545 1>}
  n34 = TextureSampleParameter2D() {
   group=Color
   parameterName=Color Mask
   sortPriority=2
   texture=/Game/Creative/Devices/Generic/Materials/Textures/T_CP_Device_GearIcon_Placed_01_D
  }
  n35 = LinearInterpolate(
   A=@n33.RGB
   B=Multiply(A=ScalarParameter() {defaultValue=0.75, group=Color, parameterName=Color_InnerRingMultiplier}, B=@n33.RGB)
   Alpha=@n34.B
  )
  n36 = Multiply(A=ScalarParameter() {defaultValue=0.25, group=Color, parameterName=Color_InnerCircleMultiplier}, B=@n33.RGB)
  n37 = LinearInterpolate(
   A=LinearInterpolate(A=LinearInterpolate(A=@n35, B=@n36, Alpha=@n34.G), B=@n33.RGB, Alpha=@n34.R)
   B=Multiply(A=VectorParameter() {group=Logo, parameterName=Logo Color, sortPriority=1, defaultValue=<rgba 1 1 1 1>}.RGB, B=@n32.RGB)
   Alpha=@n32.A
  )
  n38 = Multiply(A=@n37, B=@n27)
  n39 = ScalarParameter() {defaultValue=0.5, group=Color Curve, parameterName=Lerp Color Curve Overlay}
  Reroute(None=StaticSwitchParameter(True=LinearInterpolate(A=@n37, B=@n38, Alpha=@n39), False=@n37) {defaultValue=true, group=Color Curve, parameterName=UseColorCurve, sortPriority=0})
  Reroute(None=Multiply(A=@n17, B=ScalarParameter() {parameterName=Specular}) {constB=2})
  n40 = Normalize(VectorInput=Constant3Vector() {constant=<rgba 0 0 1 0>})
  n41 = MaterialFunctionCall() {materialFunction=&ObjectPivotPoint}
  n42 = WorldPosition()
  Reroute(None=StaticSwitchParameter(
    True=RotateAboutAxis(NormalizedRotationAxis=@n40, RotationAngle=@n2, PivotPoint=@n41.Object Pivot Location, Position=@n42.XYZ)
    False=Constant()
   ) {defaultValue=true, group=MeshRotation, parameterName=RotateAroundAxis})
  n43 = AppendVector(A=TextureCoordinate() {coordinateIndex=2}, B=ComponentMask(None=TextureCoordinate() {coordinateIndex=3}) {r=true})
  n44 = TextureSampleParameter2D() {group=Texture, parameterName=Normal, samplerType=SAMPLERTYPE_Normal, texture=/Engine/EngineMaterials/BaseFlattenNormalMap}
  StaticSwitchParameter(
   True=StaticSwitchParameter(
    True=MaterialFunctionCall(BaseNormal=@n43, AdditionalNormal=@n44.RGB) {materialFunction=&BlendAngleCorrectedNormals}.Result
    False=@n43
   ) {parameterName=UseNormalMap}
   False=StaticSwitchParameter(True=@n44.RGB, False=Constant3Vector() {constant=<rgba 0 0 1 0>}) {parameterName=UseNormalMap}
  ) {defaultValue=true, group=MeshRotation, parameterName=RotateAroundAxis}
 ]
}
